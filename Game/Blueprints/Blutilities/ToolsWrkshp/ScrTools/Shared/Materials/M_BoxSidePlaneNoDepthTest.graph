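# Unlit, translucent, two-sided material for box side planes in the ScrTools editor utilities, with depth testing disabled so it draws through scene geometry.
# A CanvasFront texture parameter, defaulting to a square button texture, is tinted by vertex color RGB.
# A square-rooted radial gradient, blended in by the Do Radial Fade parameter, multiplies vertex alpha to form a fade mask.
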Material /Game/Blueprints/Blutilities/ToolsWrkshp/ScrTools/Shared/Materials/M_BoxSidePlaneNoDepthTest
graph {
 expressions {
  MaterialExpressionConstant_1 {class=MaterialExpressionConstant, properties {r=1}}
  MaterialExpressionConvert_1 {
   class=MaterialExpressionConvert
   inputs [
    None {from=MaterialExpressionVertexColor_1, output=R}
    None {from=MaterialExpressionVertexColor_1, output=R}
    None {from=MaterialExpressionVertexColor_1, output=R}
   ]
   properties {
    nodeName=Make Vector3
    convertInputs [
     {
      type=Scalar
      expressionInput {
       expression=&self:MaterialExpressionVertexColor_1
       inputName=None
       mask=1
       maskA=0
       maskB=0
       maskG=0
       maskR=1
       outputIndex=1
      }
     }
     {
      type=Scalar
      expressionInput {
       expression=&self:MaterialExpressionVertexColor_1
       inputName=None
       mask=1
       maskA=0
       maskB=0
       maskG=1
       maskR=0
       outputIndex=2
      }
     }
     {
      type=Scalar
      expressionInput {
       expression=&self:MaterialExpressionVertexColor_1
       inputName=None
       mask=1
       maskA=0
       maskB=1
       maskG=0
       maskR=0
       outputIndex=3
      }
     }
    ]
    convertOutputs [{type=Vector3}]
   }
  }
  MaterialExpressionLinearInterpolate_1 {
   class=MaterialExpressionLinearInterpolate
   inputs [
    A {from=MaterialExpressionConstant_1}
    B {from=MaterialExpressionPower_1}
    Alpha {from=MaterialExpressionScalarParameter_4}
   ]
  }
  MaterialExpressionMaterialFunctionCall_1 {
   class=MaterialExpressionMaterialFunctionCall
   inputs [
    UV {from=MaterialExpressionTextureCoordinate_0}
    GradientStart {from=""}
    GradientEnd {from=""}
   ]
   properties {materialFunction=&MF_UI_RadialGradient}
  }
  MaterialExpressionMultiply_0 {
   class=MaterialExpressionMultiply
   inputs [
    A {from=MaterialExpressionTextureSampleParameter2D_0, output=RGB}
    B {from=MaterialExpressionConvert_1}
   ]
  }
  MaterialExpressionMultiply_2 {
   class=MaterialExpressionMultiply
   inputs [
    A {from=MaterialExpressionLinearInterpolate_1}
    B {from=MaterialExpressionVertexColor_1, output=A}
   ]
  }
  MaterialExpressionPower_1 {
   class=MaterialExpressionPower
   inputs [
    Base {from=MaterialExpressionMaterialFunctionCall_1, output=Smooth}
    Exp {from=""}
   ]
   properties {constExponent=0.5}
  }
  MaterialExpressionScalarParameter_4 {class=MaterialExpressionScalarParameter, properties {parameterName=Do Radial Fade}}
  MaterialExpressionTextureCoordinate_0 {class=MaterialExpressionTextureCoordinate}
  MaterialExpressionTextureCoordinate_1 {class=MaterialExpressionTextureCoordinate}
  MaterialExpressionTextureSampleParameter2D_0 {
   class=MaterialExpressionTextureSampleParameter2D
   inputs [
    UVs {from=MaterialExpressionTextureCoordinate_1, output=""}
    Apply View MipBias {from="", output=""}
   ]
   properties {parameterName=CanvasFront, texture=/Game/UI/Foundation/Textures/Buttons/T-Button-Square}
  }
  MaterialExpressionVertexColor_1 {class=MaterialExpressionVertexColor, inputs [], properties {}}
 }
 properties {
  bAutomaticallySetUsageInEditor=false
  bCastRayTracedShadows=false
  bDisableDepthTest=true
  blendMode=BLEND_Translucent
  shadingModel=MSM_Unlit
  twoSided=true
  usedShadingModels=Unlit
 }
}
