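# Editor utility Blueprint that holds the settings for a Snap To Target tool, likely as its scriptable tool property set.
# It is data only, with an empty event graph and defaulted variables.
# Settings cover snap side, pivot, offsets with step increments, rotation step, grid snapping, per-object axis snapping, and handling of hidden or duplicated objects.
# Others style the visual aids: bounding box, axis plane, grid and radial fade.
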EditorUtilityBlueprint /Game/Blueprints/Blutilities/ToolsWrkshp/ScrTools/SnapToTarget/BP_STPS_SnapToTarget
graphs [EventGraph {links [], nodes []}]
variables [
 Along Axis Offset {type=number}
 Along Axis Offset Increment {type=number, default=5}
 Bounding Box Color {type=Color, default=<rgba 0.9216 0 0 1>}
 Bounding Box Line Thickness {type=number, default=3}
 Duplicated Offset Amount {type=number, default=0.2}
 Grid Snap Size {type=integer, default=512}
 Grid Snapping {type=boolean}
 Ignore Duplicated {type=boolean, default=true}
 Pivot Offset {type=Vector}
 Radial Fade {type=boolean, default=true}
 Rotation Angle Increment {type=number, default=10}
 Selection Snap Axis {type=E_SnapToTarget_BoundSide, default=Bottom}
 Show Snap Axis Plane {type=boolean, default=true}
 Show Snap Grid {type=boolean, default=true}
 Snap Axis Plane Color {type=LinearColor, default=<rgba 1 1 1 0.1>}
 Snap Each Object Along Axis {type=boolean}
 Snap Grid Color {type=LinearColor, default=<rgba 1 1 1 0.1>}
 Snap Grid Line Thickness {type=integer, default=15}
 Snap To Hidden {type=boolean}
 Vertical Offset {type=number}
 Vertical Offset Increment {type=number, default=5}
]
