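# Scriptable editor tool that snaps selected level actors as a group to the surface under the cursor.
# While the preview key is held, hover traces ignore moved actors, then apply grid snapping and vertical offset to place the group bounds.
# Hotkeys route the mouse wheel to yaw rotation, vertical offset or bounds-axis movement.
# Transacted duplication, readout widgets, bounds lines and render-target grid and snap-side quads support placement.
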EditorUtilityBlueprint /Game/Blueprints/Blutilities/ToolsWrkshp/ScrTools/SnapToTarget/BP_ST_SnapToTarget
graphs [
 ActiveHover {nodes [
   n1 = ActiveHover!() -> (then, HoverCurPosWorldRay:Ray Structure) {is=title K2Node_FunctionEntry}
   n2 = GetRayStartEnd!(Ray:Ray Structure=@n1, StartDistance:double=0, EndDistance:double=0) -> (StartPoint:Vector, EndPoint:Vector)
   > n3 = Line Trace By Channel!(
    execute
    Start:Vector=@n2
    End:Vector=@n2.EndPoint
    TraceChannel:ETraceTypeQuery Enum
    bTraceComplex:bool=true
    ActorsToIgnore:Array of Actor Object References=Get Current Actors To Ignore In Trace() -> (Current Actors To Ignore In Trace:Array of Actor Object References)
    DrawDebugType:EDrawDebugTrace Enum
    bIgnoreSelf:bool
    TraceColor:Linear Color Structure
    TraceHitColor:Linear Color Structure
    DrawTime:float
   ) -> (then, OutHit:Hit Result Structure, ReturnValue:bool)
   > n4 = Set ActiveHoverOutHit(ActiveHoverOutHit:Hit Result Structure=@n3) -> (Output_Get:Hit Result Structure)
   > Branch(Condition=@n3.ReturnValue)
   > n5 = WaitForTime!(execute, self, PauseTime:double=Get PreviewMoveIntervalTime() -> (PreviewMoveIntervalTime:double)) -> (then, Active:bool)
   > Branch(Condition=@n5)
   > n6 = GetActorArrayBounds!(
    execute
    Actors:Array of Actor Object References=Get ActorsToMove() -> (ActorsToMove:Array of Actor Object References)
    bOnlyCollidingComponents:bool=Get Only Colliding Components() -> (Only Colliding Components:bool)
   ) -> (then, Center:Vector, BoxExtent:Vector)
   > Set CurrentBoundsCenter(CurrentBoundsCenter:Vector=@n6) -> (Output_Get:Vector)
   > Set CurrentBoundsExtent(CurrentBoundsExtent:Vector=@n6.BoxExtent) -> (Output_Get:Vector)
   n7 = Get Settings() -> (Settings:BP STPS Snap To Target Object Reference)
   > n8 = SnapVectorToGrid!(
    execute
    VectorToSnap:Vector=BreakHitResult!(Hit:"Hit Result Structure (by ref)"=@n4) -> (
     bBlockingHit:bool
     bInitialOverlap:bool
     Time:float
     Distance:float
     Location:Vector
     ImpactPoint:Vector
     Normal:Vector
     ImpactNormal:Vector
     PhysMat:Physical Material Object Reference
     HitActor:Actor Object Reference
     HitComponent:Primitive Component Object Reference
     HitBoneName:Name
     BoneName:Name
     HitItem:int
     ElementIndex:int
     FaceIndex:int
     TraceStart:Vector
     TraceEnd:Vector
    ).Location
    GridSize:int=Get Grid Snap Size!(self:BP STPS Snap To Target Object Reference=@n7) -> (Grid Snap Size:int=0)
    Snap On:bool=Get Grid Snapping!(self:BP STPS Snap To Target Object Reference=@n7) -> (Grid Snapping:bool)
    Ignore Up:bool=false
   ) -> (then, OutputVector:Vector)
   > n9 = Set HoveredSnappedLocation(HoveredSnappedLocation:Vector=@n8) -> (Output_Get:Vector)
   n10 = BreakVector(InVec=@n9)
   n11 = Get Vertical Offset!(self:BP STPS Snap To Target Object Reference=Get Settings()) -> (Vertical Offset:double)
   n12 = Get CurrentBoundsCenter() -> (CurrentBoundsCenter:Vector)
   > TransformActorGroup!(
    execute
    self
    NewLocation:Vector=MakeVector(X=@n10, Y=@n10.Y, Z=float + float(A=@n10.Z, B=@n11))
    GroupBoundCenter:Vector=@n12
    GroupBoundExtent:Vector=Get CurrentBoundsExtent() -> (CurrentBoundsExtent:Vector)
    Rotation:Rotator="0, 0, 0"
   ) -> (then)
   > n13 = Sequence!(execute) -> (then_0, then_1, then_2:Exec, then_3:Exec)
   n14 = Get Selection Snap Axis!(self:BP STPS Snap To Target Object Reference=Get Settings()) -> (Selection Snap Axis:E_SnapToTarget_BoundSide Enum=NewEnumerator0)
   > n15 = Switch on E_SnapToTarget_BoundSide!(execute, Selection:E_SnapToTarget_BoundSide Enum=@n14) -> (
    NewEnumerator0:Exec
    NewEnumerator1:Exec
    NewEnumerator2:Exec
    NewEnumerator3:Exec
    NewEnumerator4:Exec
    NewEnumerator5:Exec
    NewEnumerator6:Exec
   ) {is=title K2Node_SwitchEnum}
   Branch(
    execute=[@n15.NewEnumerator0, @n15.NewEnumerator1, @n15.NewEnumerator2, @n15.NewEnumerator3, @n15.NewEnumerator4, @n15.NewEnumerator5]
    Condition=Get Show Snap Axis Plane!(self:BP STPS Snap To Target Object Reference=Get Settings()) -> (Show Snap Axis Plane:bool)
   )
   n16 = Select!(
    NewEnumerator0:int=2
    NewEnumerator1:int=3
    NewEnumerator2:int=5
    NewEnumerator3:int=4
    NewEnumerator4:int=1
    NewEnumerator5:int=0
    NewEnumerator6:int=6
    Index:E_SnapToTarget_BoundSide Enum=@n14
   ) -> (ReturnValue:int)
   > n17 = Get Bbox Side Points!(execute, BBox Origin:Vector=Get CurrentBoundsCenter(), BBox Extent:Vector=Get CurrentBoundsExtent(), Side:int=@n16) -> (then, Point 0:Vector, Point 1:Vector, Point 2:Vector, Point 3:Vector) {is=title K2Node_CallFunction}
   > SetQuadPoints!(
    execute
    self:Scriptable Tool Quad Object Reference=Get SnapSide_PlaneQuadComponent() -> (SnapSide_PlaneQuadComponent:Scriptable Tool Quad Object Reference)
    A:Vector=@n17
    B:Vector=@n17.Point 1
    C:Vector=@n17.Point 2
    D:Vector=@n17.Point 3
   ) -> (then)
   n18 = GetBoxFaceCenter!(
    Box:Box Structure=MakeBoxFromCenterExtents!(Center:Vector=Get CurrentBoundsCenter(), Extents:Vector=Get CurrentBoundsExtent()) -> (ReturnValue:Box Structure)
    FaceIndex:int=@n16
   ) -> (FaceNormal:Vector, ReturnValue:Vector)
   > SetQuadNormals!(
    execute
    self:Scriptable Tool Quad Object Reference=Get SnapSide_PlaneQuadComponent()
    A:Vector=@n18
    B:Vector=@n18
    C:Vector=@n18
    D:Vector=@n18
   ) -> (then)
   > UpdateResource(self:Canvas Render Target 2D Object Reference=Get SnapSideText Render Target() -> (SnapSideText Render Target:Canvas Render Target 2D Object Reference))
   n19 = Get Settings()
   Branch(
    execute=@n13.then_1
    Condition=AND Boolean(A=Get Show Snap Grid!(self:BP STPS Snap To Target Object Reference=@n19) -> (Show Snap Grid:bool), B=Get Grid Snapping(self=@n19))
   )
   > UpdateResource(self=Get Grid Draw Render Target() -> (Grid Draw Render Target:Canvas Render Target 2D Object Reference))
   n20 = Get Settings()
   DrawBBox!(
    execute=@n13.then_2
    BoundOrigin:Vector=Get CurrentBoundsCenter()
    BoundExtent:Vector=Get CurrentBoundsExtent()
    Box Lines:Array of Scriptable Tool Line Object References=Get Box Lines() -> (Box Lines:Array of Scriptable Tool Line Object References)
    Color:Color Structure=Get Bounding Box Color!(self:BP STPS Snap To Target Object Reference=@n20) -> (Bounding Box Color:Color Structure)
    Thickness:float=Get Bounding Box Line Thickness!(self:BP STPS Snap To Target Object Reference=@n20) -> (Bounding Box Line Thickness:double)
    DepthBias:float=0
   ) -> (then)
   n21 = Get PreMoveBBoxCenter() -> (PreMoveBBoxCenter:Vector)
   SetLineEndPoints!(
    execute=@n13.then_3
    self:Scriptable Tool Line Object Reference=Get Dist Line() -> (Dist Line:Scriptable Tool Line Object Reference)
    Start:Vector=@n21
    End:Vector=@n12
   ) -> (then)
   > n22 = GetDistConverted!(execute, self, DistInCm:double="Distance (Vector)"!(V1:Vector=@n12, V2:Vector=@n21) -> (ReturnValue:double)) -> (then, ConvertedDist:double, Units:String)
   > n23 = Set CurDistanceText(CurDistanceText:Text="To Text (Float)"(Value=@n22, MinimumIntegralDigits=1, MaximumIntegralDigits=324, MinimumFractionalDigits=0, MaximumFractionalDigits=2)) -> (Output_Get:Text)
   > SetInfoText!(execute, self, DistanceValue:Text=@n23, WheelText:Text=Get CurWheelText() -> (CurWheelText:Text)) -> (then)
   > Return Node()
   Format Text!(Format="LOCTEXT(\"Dist: {d}\\r\\n{units}\")", d:double=@n22, units:String=@n22.Units) -> (Result)
  ]}
 AddPivotAdjustWidget {nodes [
   AddPivotAdjustWidget() {is=title K2Node_FunctionEntry}
   > n1 = GetSelectedLevelActors(self=Get EditorActorSubsystem())
   > n2 = GetActorArrayBounds(Actors=@n1, bOnlyCollidingComponents=false)
   n3 = Get Settings()
   > CreateTRSGizmo!(
    execute
    self:Scriptable Interactive Tool Object Reference
    Identifier:String=Get PivotGizmoIdentifier() -> (PivotGizmoIdentifier:String)
    InitialTransform:Transform=MakeTransform(Location=vector + vector(A=@n2), Rotation="0, 0, 0")
    GizmoOptions:Scriptable Tool Gizmo Options Structure=Make Scriptable Tool Gizmo Options!(
     GizmoMode:EScriptableToolGizmoMode Enum=TranslationOnly
     TranslationParts:Bitmask Integer=23
     RotationParts:Bitmask Integer=255
     ScaleParts:Bitmask Integer=255
     CoordSystem:EScriptableToolGizmoCoordinateSystem Enum=Local
     bSnapTranslation:bool=Get Grid Snapping(self=@n3)
     bSnapRotation:bool=false
     bRepositionable:bool=False
     bAllowNegativeScaling:bool=false
    ) -> (ScriptableToolGizmoOptions:Scriptable Tool Gizmo Options Structure) {is=title K2Node_MakeStruct}
   ) -> (Success:Exec, Failure:Exec)
   vector * vector(A=Get Pivot Offset!(self:BP STPS Snap To Target Object Reference=Get Settings()) -> (Pivot Offset:Vector), B:Vector="1,1,0") -> (ReturnValue=~)
  ]}
 EventGraph {nodes [
   Event OnScriptSetup()
   > ScriptSetup()
   n1 = Event OnScriptTick()
   > Set CurTime(CurTime=float + float(A=@n1.DeltaTime, B=Get CurTime()))
   > n2 = Branch(Condition=Get FirstInitRun())
   Set FirstInitRun(execute=@n2.else)
   > n3 = Sequence!(execute) -> (then_0, then_1, then_2)
   n4 = Get Header Widget()
   > n5 = AddExpandableBox(self=@n4)
   > n6 = Set Readout Box(Readout Box:BP UW Expandable Box Object Reference=@n5) -> (Output_Get:BP UW Expandable Box Object Reference)
   > SetUpBox(
    self=@n6
    Name=Get Readout Box Label!(self:BP UW Shared Header Widget Object Reference=@n4) -> (Readout Box Label:Text)
    OpenByDefault=Get ReadoutOpenOnExit!(self:BP STPS Shared Widget State Object Reference=Get SharedState Settings() -> (SharedState Settings:BP STPS Shared Widget State Object Reference)) -> (ReadoutOpenOnExit:bool)
   )
   > n7 = Sequence!(execute) -> (then_0, then_1)
   n8 = Set Distance Line(execute=@n9.then, Distance Line:BP UW Readout Line Object Reference=@n9) -> (Output_Get:BP UW Readout Line Object Reference)
   n9 = SetupLine!(
    execute=@n7.then_0
    self:BP UW Expandable Box Object Reference=Get Readout Box() -> (Readout Box:BP UW Expandable Box Object Reference)
    CurrentLine:BP UW Readout Line Object Reference=@n8
    DynamicLine:Vertical Box Object Reference
    Label:Text=Get Widget Distance Line Label() -> (Widget Distance Line Label:Text)
    Value:Text=Get Widget Distance Line Value() -> (Widget Distance Line Value:Text)
   ) -> (then, Readout Line:BP UW Readout Line Object Reference)
   n10 = Set Angle Line(execute=@n11.then, Angle Line:BP UW Readout Line Object Reference=@n11) -> (Output_Get:BP UW Readout Line Object Reference)
   n11 = SetupLine(
    execute=@n7.then_1
    self=Get Readout Box()
    CurrentLine=@n10
    Label=Get Widget Angle Rotated Line Label() -> (Widget Angle Rotated Line Label:Text)
    Value=Get Widget Angle Rotated Line Value() -> (Widget Angle Rotated Line Value:Text)
   )
   n12 = AddExpandableBox(execute=@n3.then_1, self=Get Header Widget())
   > n13 = Set Shortcut Box(Shortcut Box:BP UW Expandable Box Object Reference=@n12) -> (Output_Get:BP UW Expandable Box Object Reference)
   > n14 = SetUpBox(
    self=@n13
    Name=Get Shortcut Box Label!(self:BP UW Shared Header Widget Object Reference=Get Header Widget()) -> (Shortcut Box Label:Text)
    OpenByDefault=Get ShortcutsOpenOnExit!(self:BP STPS Shared Widget State Object Reference=Get SharedState Settings()) -> (ShortcutsOpenOnExit:bool)
   )
   n15 = Set ShortcutLines(execute=@n16.then, ShortcutLines:Array of BP UW Readout Line Object References=@n16) -> (Output_Get:Array of BP UW Readout Line Object References)
   n16 = AddShortcuts!(
    execute=@n14.then
    self:BP UW Expandable Box Object Reference=Get Shortcut Box() -> (Shortcut Box:BP UW Expandable Box Object Reference)
    ShortcutMap:Map of Strings to Strings=Get Shortcut Map() -> (Shortcut Map:Map of Strings to Strings)
    ExistingShortcutLines:Array of BP UW Readout Line Object References=@n15
    ShortcutMapKeys:Array of Strings=Get Shortcut Map Keys() -> (Shortcut Map Keys:Array of Strings)
   ) -> (then, AllShortcutLines:Array of BP UW Readout Line Object References)
   n17 = AddExpandableBox(execute=@n3.then_2, self=Get Header Widget())
   > n18 = Set Actions Box(Actions Box=@n17)
   > SetUpBox(self=@n18, Name=Get Actions Box Label(self=Get Header Widget()))
   > n19 = Sequence()
   > n20 = Construct Object from Class(Class=/Game/Blueprints/Blutilities/ToolsWrkshp/ScrTools/SnapToTarget/SupportFiles/BP_UW_DupButton.BP_UW_DupButton_C) -> (ReturnValue:BP UW Dup Button Object Reference)
   > n21 = Set DupButton(DupButton:BP UW Dup Button Object Reference=@n20) -> (Output_Get:BP UW Dup Button Object Reference)
   > AddChildToVerticalBox(self=Get DynamicEntries(self=Get Actions Box()), Content=@n21)
   > Bind Event to Widget Dupe Clicked(self:BP UW Dup Button Object Reference=Get DupButton() -> (DupButton:BP UW Dup Button Object Reference), Delegate=Create Event()) {is=title K2Node_AddDelegate}
   n22 = Construct Object from Class(
    execute=@n19.then_1
    Class=/Game/Blueprints/Blutilities/ToolsWrkshp/ScrTools/SnapToTarget/SupportFiles/BP_UW_SnapEachObjectButton.BP_UW_SnapEachObjectButton_C
   ) -> (ReturnValue:BP UW Snap Each Object Button Object Reference)
   > n23 = Set Snap Button(Snap Button:BP UW Snap Each Object Button Object Reference=@n22) -> (Output_Get:BP UW Snap Each Object Button Object Reference)
   > AddChildToVerticalBox(self=Get DynamicEntries(self=Get Actions Box()), Content=@n23)
   > Bind Event to Widget Snap Each Clicked(
    self:BP UW Snap Each Object Button Object Reference=Get Snap Button() -> (Snap Button:BP UW Snap Each Object Button Object Reference)
    Delegate=Create Event()
   ) {is=title K2Node_AddDelegate}
   n24 = Event OnScriptShutdown()
   > ScriptShutdown!(execute, self, ShutdownType:EToolShutdownType Enum=@n24.ShutdownType) -> (then)
  ]}
 FindCoincidentActors {nodes [
   FindCoincidentActors!() -> (then, ActorsToTest:Array of Actor Object References) {is=title K2Node_FunctionEntry}
   > n1 = Sequence()
   > n2 = GetAllLevelActors(self=Get EditorActorSubsystem())
   > Set AllLevelActorsExceptSelected(AllLevelActorsExceptSelected:Array of Actor Object References=@n2) -> (Output_Get:Array of Actor Object References)
   > n3 = For Each Loop(Array:Array of Actor Object References=Get ActorsToTest() -> (ActorsToTest:Array of Actor Object References)) -> (Array Element:Actor Object Reference)
   > Remove Item!(
    execute
    TargetArray:Array of Actor Object References=Get AllLevelActorsExceptSelected() -> (AllLevelActorsExceptSelected:Array of Actor Object References)
    Item:"Actor Object Reference (by ref)"=@n3
   ) -> (then, ReturnValue:bool)
   n4 = For Each Loop(Exec=@n1.then_1, Array=Get ActorsToTest())
   n5 = Get ActorsToTest()
   Branch(execute=@n4.Completed, Condition="Equal (Integer)"(A=Length(TargetArray=@n5), B=Get NumSelNearlyEqual() -> (NumSelNearlyEqual:int=0)))
   > n6 = For Each Loop(Array=Get ActorsToTest())
   > Destroy Actor(self=@n6)
   n7 = FindNearestActor!(Origin:Vector=Get Actor Location(self=@n4), ActorsToCheck:Array of Actor Object References=Get AllLevelActorsExceptSelected()) -> (Distance:float, ReturnValue:Actor Object Reference)
   Branch(
    execute=@n4.LoopBody
    Condition="Nearly Equal (Transform)"!(
     A:"Transform (by ref)"=Get Actor Transform(self=@n4)
     B:"Transform (by ref)"=Get Actor Transform(self=@n7.ReturnValue)
     LocationTolerance:float=0.0001
     RotationTolerance:float=0.0001
     Scale3DTolerance:float=0.0001
    ) -> (ReturnValue:bool)
   )
   > Set NumSelNearlyEqual(NumSelNearlyEqual:int=int + int(A=Get NumSelNearlyEqual() -> (NumSelNearlyEqual=0))) -> (Output_Get:int=0)
  ]}
 Get Bounds Edge Center Locations {nodes [
   n1 = Get Bounds Edge Center Locations!() -> (then, BoundCenter:Vector=~, BoundExtents:Vector=~) {is=title K2Node_FunctionEntry}
   n2 = BreakVector(InVec=@n1)
   > Set CenterX(CenterX:double=@n2) -> (Output_Get:double)
   > Set CenterY(CenterY:double=@n2.Y) -> (Output_Get:double)
   > Set CenterZ(CenterZ:double=@n2.Z) -> (Output_Get:double)
   n3 = BreakVector(InVec=@n1.BoundExtents)
   > Set ExtentX(ExtentX:double=@n3) -> (Output_Get:double)
   > Set ExtentY(ExtentY:double=@n3.Y) -> (Output_Get:double)
   > Set ExtentZ(ExtentZ:double=@n3.Z) -> (Output_Get:double)
   n4 = Get CenterX() -> (CenterX:double)
   n5 = Get ExtentX() -> (ExtentX:double)
   n6 = Get CenterY() -> (CenterY:double)
   n7 = Get ExtentY() -> (ExtentY:double)
   n8 = Get CenterY()
   n9 = Get ExtentY()
   n10 = Get CenterX()
   n11 = Get ExtentX()
   n12 = Get CenterZ() -> (CenterZ:double)
   n13 = Get ExtentZ() -> (ExtentZ:double)
   n14 = Get CenterZ()
   n15 = Get ExtentZ()
   > Return Node!(
    execute
    MaxX:Vector=MakeVector(X=float + float(A=@n4, B=@n5), Y=Get CenterY(), Z=Get CenterZ())
    MinX:Vector=MakeVector(X=float - float(A=@n10, B=@n11), Y=Get CenterY(), Z=Get CenterZ())
    MaxY:Vector=MakeVector(X=Get CenterX(), Y=float + float(A=@n8, B=@n9), Z=Get CenterZ())
    MinY:Vector=MakeVector(X=Get CenterX(), Y=float - float(A=@n6, B=@n7), Z=Get CenterZ())
    MaxZ:Vector=MakeVector(X=Get CenterX(), Y=Get CenterY(), Z=float + float(A=@n14, B=@n15))
    MinZ:Vector=MakeVector(X=Get CenterX(), Y=Get CenterY(), Z=float - float(A=@n12, B=@n13))
   )
  ]}
 GetActorsToIgnore {nodes [
   GetActorsToIgnore() {is=title K2Node_FunctionEntry}
   > Set ActorsToIgnoreInTrace(ActorsToIgnoreInTrace:Array of Actor Object References=Get ActorsToMove()) -> (Output_Get:Array of Actor Object References)
   > n1 = Sequence!(execute) -> (then_0, then_1, then_2:Exec)
   > Branch(Condition=Get Ignore Duplicated!(self:BP STPS Snap To Target Object Reference=Get Settings()) -> (Ignore Duplicated:bool))
   > Append Array!(
    execute
    TargetArray:Array of Actor Object References=Get ActorsToIgnoreInTrace() -> (ActorsToIgnoreInTrace:Array of Actor Object References)
    SourceArray:Array of Actor Object References=Get ActorsDuplicatedFromSnapTool() -> (ActorsDuplicatedFromSnapTool:Array of Actor Object References)
   ) -> (then)
   n2 = Branch(execute=@n1.then_1, Condition=Get Snap To Hidden!(self:BP STPS Snap To Target Object Reference=Get Settings()) -> (Snap To Hidden:bool))
   n3 = GetActorsTempHidden!(execute=@n2.else) -> (then, ActorTempHiddenMap:Map of Actor Object References to Booleans, OnlyTempHidden:Array of Actor Object References)
   > Append Array(TargetArray=Get ActorsToIgnoreInTrace(), SourceArray=@n3.OnlyTempHidden)
   Return Node!(execute=@n1.then_2, ActorsToIgnoreInTrace:Array of Actor Object References=Get ActorsToIgnoreInTrace())
  ]}
 GetBoxSideDirectionVector {nodes [
   n1 = GetBoxSideDirectionVector!() -> (then, BoundCenter:Vector=~, BoundExtents:Vector=~) {is=title K2Node_FunctionEntry}
   > n2 = Get Bounds Edge Center Locations!(execute, self, BoundCenter:Vector=@n1, BoundExtents:Vector=@n1.BoundExtents) -> (then, MaxX:Vector, MinX:Vector, MaxY:Vector, MinY:Vector, MaxZ:Vector, MinZ:Vector) {is=title K2Node_CallFunction}
   n3 = Select(
    NewEnumerator0:Vector=@n2.MinY
    NewEnumerator1:Vector=@n2.MaxY
    NewEnumerator2:Vector=@n2
    NewEnumerator3:Vector=@n2.MinX
    NewEnumerator4:Vector=@n2.MaxZ
    NewEnumerator5:Vector=@n2.MinZ
    NewEnumerator6:Vector="0, 0, 0"
    Index=Get Selection Snap Axis(self=Get Settings()) -> (Selection Snap Axis=NewEnumerator0)
   ) -> (ReturnValue:Vector)
   n4 = Get BoundCenter() -> (BoundCenter:Vector)
   > Return Node!(
    execute
    Normalized Dir:Vector=Normalize!(A:Vector=vector - vector(A=@n3, B=@n4), Tolerance:float=0.0001) -> (ReturnValue:Vector) {is=title K2Node_CallFunction}
    Start Point:Vector=@n3
   )
  ]}
 GetDistConverted {nodes [
   n1 = GetDistConverted!() -> (then, DistInCm:double=~) {is=title K2Node_FunctionEntry}
   n2 = MakeLiteralString(Value=cm/s)
   n3 = Find(
    TargetMap:"Map of Strings to Float (double-precision)s"=Get DistConversionValues() -> (DistConversionValues:"Map of Strings to Float (double-precision)s")
    Key=@n2
   ) -> (Value:double)
   n4 = SafeDivide(A=SafeDivide(A=@n1, B=100), B=@n3)
   n5 = Truncate(A=float * float(A=SelectFloat(A=@n4, B=@n1, bPickA=@n3.ReturnValue), B=100) {is=title K2Node_PromotableOperator})
   > Return Node!(
    execute
    ConvertedDist:double=float / float(A="To Float (Integer)"(InInt=@n5), B=100)
    Units:String=SelectString(A=@n2, B=Centimeters, bPickA=@n3.ReturnValue)
   )
  ]}
 HoverSequenceTest {nodes [
   HoverSequenceTest!() -> (then, CurrentPos:Input Device Ray Structure, Modifiers_bShiftDown:bool=~, Modifiers_bCtrlDown:bool=~, Modifiers_bAltDown:bool=~) {is=title K2Node_FunctionEntry}
   > Return Node!(
    execute
    ReturnValue_bHit:bool=true
    ReturnValue_HitDepth:double=340282346638528859811704183484516925440
    ReturnValue_HitNormal:Vector="0,0,1"
    ReturnValue_bHasHitNormal:bool=False
    ReturnValue_HitIdentifier:int=0
    ReturnValue_HitObject:Object Reference
   )
  ]}
 IfHitByClick {nodes [
   n1 = IfHitByClick!() -> (
    then
    ClickPos_WorldRay:Ray Structure
    ClickPos_bHas2D:bool=~
    ClickPos_ScreenPosition:Vector 2D Structure
    MouseButton:EScriptableToolMouseButton Enum=LeftButton
   ) {is=title K2Node_FunctionEntry}
   > n2 = GetActorsToIgnore!(execute, self) -> (then, ActorsToIgnoreInTrace:Array of Actor Object References) {is=title K2Node_CallFunction}
   n3 = GetRayStartEnd(Ray=@n1, StartDistance=0, EndDistance=0)
   > n4 = Line Trace By Channel(Start=@n3, End=@n3.EndPoint, bTraceComplex=true, ActorsToIgnore=@n2)
   > Set TraceOutHit(TraceOutHit:Hit Result Structure=@n4) -> (Output_Get:Hit Result Structure)
   > Return Node(
    ReturnValue_bHit=@n4.ReturnValue
    ReturnValue_HitDepth=340282346638528859811704183484516925440
    ReturnValue_HitNormal="0,0,1"
    ReturnValue_bHasHitNormal=False
    ReturnValue_HitIdentifier=0
   )
  ]}
 MouseWheelAction {nodes [
   n1 = MouseWheelAction!() -> (then, IsDown:bool=~) {is=title K2Node_FunctionEntry}
   > Set LastWheelTime(LastWheelTime:double=Get CurTime()) -> (Output_Get:double)
   n2 = Get Rotation Angle Increment!(self:BP STPS Snap To Target Object Reference=Get Settings()) -> (Rotation Angle Increment:double)
   > Set Mouse Wheel Delta(Mouse Wheel Delta:double=SelectFloat(A=float * float(A=@n2, B=-1), B=@n2, bPickA=@n1)) -> (Output_Get:double)
   > n3 = Branch(Condition=Get PreviewKeyDown() -> (PreviewKeyDown:bool))
   > n4 = GetActorArrayBounds(Actors=Get ActorsToMove(), bOnlyCollidingComponents=false)
   > n5 = Set RotPivot(RotPivot:Vector=@n4) -> (Output_Get:Vector)
   > Set Rot Pivot With Offset(Rot Pivot With Offset:Vector=vector + vector(A=@n5)) -> (Output_Get:Vector)
   > n6 = Sequence()
   n7 = Get PreMoveActorLocationsWithRotOffset() -> (PreMoveActorLocationsWithRotOffset:Array of Vectors)
   > Set PreMoveNumPoints(PreMoveNumPoints:int=Length(TargetArray:Array of Vectors=@n7)) -> (Output_Get:int=0)
   > n8 = For Each Loop(Array:Array of Vectors=@n7) -> (Array Element:Vector)
   n9 = Set PreMoveCentroid(
    execute=@n8.Completed
    PreMoveCentroid:Vector=vector / vector(A:Vector=Get PreMoveSumVector() -> (PreMoveSumVector:Vector), B:int=Get PreMoveNumPoints() -> (PreMoveNumPoints:int=0)) -> (ReturnValue:Vector)
   ) -> (Output_Get:Vector)
   > Set Pre Move Centroid With Offset(Pre Move Centroid With Offset:Vector=vector + vector(A=@n9)) -> (Output_Get:Vector)
   Set PreMoveSumVector(execute=@n8.LoopBody, PreMoveSumVector:Vector=vector + vector(A=@n8, B=Get PreMoveSumVector())) -> (Output_Get:Vector)
   n10 = For Each Loop(Exec=@n6.then_2, Array:Array of Actor Object References=Get ActorsToMove()) -> (Array Element:Actor Object Reference)
   n11 = Set Cur Rotation Amount(
    execute=@n10.Completed
    Cur Rotation Amount:double=float + float(A=Get Cur Rotation Amount() -> (Cur Rotation Amount:double), B=Get Mouse Wheel Delta() -> (Mouse Wheel Delta:double))
   ) -> (Output_Get:double)
   > Set CurWheelText(CurWheelText:Text="To Text (Float)"(Value=@n11, MinimumIntegralDigits=1, MaximumIntegralDigits=324, MinimumFractionalDigits=0, MaximumFractionalDigits=3)) -> (Output_Get:Text)
   > SetInfoText(DistanceValue=Get CurDistanceText() -> (CurDistanceText:Text), WheelText=Get CurWheelText())
   Set CurActor(execute=@n10.LoopBody, CurActor:Actor Object Reference=@n10) -> (Output_Get:Actor Object Reference)
   > Set CurActorIndex(CurActorIndex:int=@n10.Array Index) -> (Output_Get:int=0)
   n12 = vector - vector(
    A=Get Actor Location(self=Get CurActor() -> (CurActor:Actor Object Reference))
    B=Get Rot Pivot With Offset() -> (Rot Pivot With Offset:Vector)
   )
   n13 = Get Mouse Wheel Delta()
   > Set NewPostRotLocation(NewPostRotLocation:Vector=vector + vector(
     A=Rotate Vector Around Axis!(InVect:Vector=@n12, AngleDeg:float=@n13, Axis:Vector="0,0,1") -> (ReturnValue:Vector)
     B=Get Rot Pivot With Offset()
    )) -> (Output_Get:Vector)
   n14 = BreakRotator(InRot=Get Actor Rotation!(self:Actor Object Reference=Get CurActor()) -> (ReturnValue:Rotator) {is=title K2Node_CallFunction})
   n15 = Get Mouse Wheel Delta()
   > Set Actor Location And Rotation!(
    execute
    self:Actor Object Reference=Get CurActor()
    NewLocation:Vector=Get NewPostRotLocation() -> (NewPostRotLocation:Vector)
    NewRotation:Rotator=MakeRotator(Roll=@n14, Pitch=@n14.Pitch, Yaw=float + float(A=@n14.Yaw, B=@n15))
    bSweep:bool=false
    bTeleport:bool=false
   ) -> (then, SweepHitResult:Hit Result Structure, ReturnValue:bool) {is=title K2Node_CallFunction}
   n16 = For Each Loop(Exec=@n6.then_1, Array:Array of Vectors=Get PreMoveActorLocationsWithRotOffset()) -> (Array Element:Vector)
   n17 = vector - vector(A=@n16, B=Get Pre Move Centroid With Offset() -> (Pre Move Centroid With Offset:Vector))
   n18 = Get Mouse Wheel Delta()
   > Set Array Elem!(
    execute
    TargetArray:Array of Vectors=Get PreMoveActorLocationsWithRotOffset()
    Index:int=@n16.Array Index
    Item:"Vector (by ref)"=vector + vector(A=Rotate Vector Around Axis(InVect=@n17, AngleDeg=@n18, Axis="0,0,1"), B=Get Pre Move Centroid With Offset())
    bSizeToFit:bool=false
   ) -> (then)
   n19 = Branch(execute=@n3.else, Condition=Get MoveVertKeyDown() -> (MoveVertKeyDown:bool))
   > Set TempVertMoveOffset(TempVertMoveOffset:double=Get VertOffsetAtStartOfVertMove() -> (VertOffsetAtStartOfVertMove:double)) -> (Output_Get:double)
   n20 = Get Settings()
   n21 = Get Vertical Offset Increment!(self:BP STPS Snap To Target Object Reference=@n20) -> (Vertical Offset Increment:double)
   n22 = float * float(A=Get Mouse Wheel Delta(), B=@n21)
   > Set This Move Delta(This Move Delta:double=@n22) -> (Output_Get:double)
   n23 = Get Settings()
   > Set Vertical Offset!(execute, Vertical Offset:double=float + float(A=Get Vertical Offset(self=@n23), B=@n22), self:BP STPS Snap To Target Object Reference=@n20) -> (then, Output_Get:double)
   n24 = Get Mouse Wheel Delta()
   > Set TempVertMoveOffset(TempVertMoveOffset=float + float(A=Get TempVertMoveOffset() -> (TempVertMoveOffset:double), B=float * float(A=@n24, B=@n21)))
   Branch(execute=@n19.else, Condition=Get Move Along Axis Key Down() -> (Move Along Axis Key Down:bool))
   > n25 = GetActorArrayBounds(Actors=Get ActorsToMove(), bOnlyCollidingComponents=Get Only Colliding Components())
   > n26 = GetBoxSideDirectionVector!(execute, self, BoundCenter:Vector=@n25, BoundExtents:Vector=@n25.BoxExtent) -> (then, Normalized Dir:Vector, Start Point:Vector) {is=title K2Node_CallFunction}
   > n27 = For Each Loop(Array:Array of Actor Object References=Get ActorsToMove()) -> (Array Element:Actor Object Reference)
   n28 = float * float(
    A=Get Mouse Wheel Delta()
    B=Get Along Axis Offset Increment!(self:BP STPS Snap To Target Object Reference=Get Settings()) -> (Along Axis Offset Increment:double)
   )
   > Set Actor Location!(
    execute
    self:Actor Object Reference=@n27
    NewLocation:Vector=vector + vector(A=Get Actor Location(self=@n27), B=vector * vector(A=@n26, B:double=@n28) -> (ReturnValue=~))
    bSweep:bool=false
    bTeleport:bool=false
   ) -> (then, SweepHitResult:Hit Result Structure, ReturnValue:bool) {is=title K2Node_CallFunction}
   > Set MoveCommitted(MoveCommitted:bool=true) -> (Output_Get:bool)
   n29 = Get PreMoveActorLocationsWithRotOffset()
   Set PreMoveNumPoints(PreMoveNumPoints=Length(TargetArray=@n29)) -> (Output_Get=0)
   > n30 = For Each Loop(Array:Array of Vectors=@n29) -> (Array Element:Vector)
   Set PreMoveCentroid(execute=@n30.Completed, PreMoveCentroid=vector / vector(A=Get PreMoveSumVector(), B=Get PreMoveNumPoints() -> (PreMoveNumPoints=0)))
   Set PreMoveSumVector(execute=@n30.LoopBody, PreMoveSumVector=vector + vector(A=@n30, B=Get PreMoveSumVector()))
   Set Cur Rotation Amount(Cur Rotation Amount=float + float(A=Get Cur Rotation Amount(), B=Get Mouse Wheel Delta()))
   > Set Rot Pivot With Offset(Rot Pivot With Offset=vector - vector(A=Get PreMoveBBoxCenter()))
   > n31 = For Each Loop(Array:Array of Actor Object References=Get ActorsToMove()) -> (Array Element:Actor Object Reference)
   PrintString!(execute=@n31.Completed, InString=MouseWheelActionFinished, bPrintToScreen, bPrintToLog, TextColor, Duration, Key) -> (then)
   n32 = "Get (a ref)"(Array:Array of Vectors=Get PreMoveActorLocationsWithRotOffset(), Dimension 1=@n31.Array Index) -> (Output:"Vector (by ref)")
   n33 = Get Rot Pivot With Offset()
   n34 = vector + vector(
    A=Get Rot Pivot With Offset()
    B=Rotate Vector Around Axis(InVect=vector - vector(A=@n32, B=@n33), AngleDeg=Get Mouse Wheel Delta(), Axis="0,0,1")
   )
   Set Actor Location(execute=@n31.LoopBody, self=@n31, NewLocation=@n34, bSweep=false, bTeleport=false)
   > Set Array Elem(TargetArray=Get PreMoveActorLocationsWithRotOffset(), Index=@n31.Array Index, Item=@n34, bSizeToFit=false)
   Get Pivot Offset(self=Get Settings())
   Get PreMoveBBoxCenter()
   Get PreMoveBBoxCenter()
   Get RotPivot() -> (RotPivot:Vector)
   Get Pivot Offset(self=Get Settings())
   Get Pivot Offset(self=Get Settings())
   Get RotPivot()
   Get PreMoveCentroid() -> (PreMoveCentroid:Vector)
   Get PreMoveCentroid()
  ]}
 MouseWheelCaptureCheck {nodes [
   n1 = MouseWheelCaptureCheck!() -> (
    then
    InputDeviceState_bShiftKeyDown:bool=~
    InputDeviceState_bAltKeyDown:bool=~
    InputDeviceState_bCtrlKeyDown:bool=~
    InputDeviceState_bCmdKeyDown:bool=~
    InputDeviceState_Keyboard:Keyboard Input Device State Structure
    InputDeviceState_Mouse:Mouse Input Device State Structure
   ) {is=title K2Node_FunctionEntry}
   n2 = XOR Boolean!(A:bool=Get PreviewKeyDown(), B:bool=Get MoveVertKeyDown()) -> (ReturnValue:bool)
   n3 = Get Move Along Axis Key Down()
   > Return Node!(execute, ReturnValue:bool=AND Boolean(A=NOT Boolean(A=@n1.InputDeviceState_bAltKeyDown), B=XOR Boolean(A=@n2, B=@n3)))
  ]}
 MouseWheelDown {nodes [
   MouseWheelDown!() -> (then, CurrentPos:Input Device Ray Structure, Modifiers:Scriptable Tool Modifier States Structure) {is=title K2Node_FunctionEntry}
   n1 = Get PreviewKeyDown()
   n2 = Get MoveVertKeyDown()
   > Branch(Condition=XOR Boolean(A=XOR Boolean(A=@n1, B=@n2), B=Get Move Along Axis Key Down()))
   > MouseWheelAction!(execute, self, IsDown:bool=true) -> (then) {is=title K2Node_CallFunction}
  ]}
 MouseWheelUp {nodes [
   MouseWheelUp!() -> (then, CurrentPos:Input Device Ray Structure, Modifiers:Scriptable Tool Modifier States Structure) {is=title K2Node_FunctionEntry}
   n1 = Get PreviewKeyDown()
   n2 = Get MoveVertKeyDown()
   > Branch(Condition=XOR Boolean(A=XOR Boolean(A=@n1, B=@n2), B=Get Move Along Axis Key Down()))
   > MouseWheelAction(IsDown=false)
  ]}
 On Canvas Grid Update {nodes [
   On Canvas Grid Update!() -> (then, Canvas:Canvas Object Reference, Width:int, Height:int) {is=title K2Node_FunctionEntry}
   n1 = Get Settings()
   > Is Valid(InputObject=@n1)
   n2 = Get Settings()
   > n3 = Branch(Condition=AND Boolean(A=Get Show Snap Grid(self=@n1), B=Get Grid Snapping(self=@n2)))
   n4 = Get Settings()
   > Scr Tools Draw Grid!(
    execute
    Location:Vector=Get HoveredSnappedLocation() -> (HoveredSnappedLocation:Vector)
    Canvas For Grid:Canvas Object Reference=Get Canvas() -> (Canvas:Canvas Object Reference)
    Scriptable Tool Quad:Scriptable Tool Quad Object Reference=Get Grid Quad Component() -> (Grid Quad Component:Scriptable Tool Quad Object Reference)
    Snap Grid Size:int=Get Grid Snap Size(self=@n4) -> (Grid Snap Size=0)
    Desired Num Grid Lines:int=4
    Canvas Width:int=Get Width() -> (Width:int=0)
    Canvas Height:int=Get Height() -> (Height:int=0)
    RenderColor:Linear Color Structure=Get Snap Grid Color!(self:BP STPS Snap To Target Object Reference=@n4) -> (Snap Grid Color:Linear Color Structure)
    Thickness:float="To Float (Integer)"(InInt=Get Snap Grid Line Thickness!(self:BP STPS Snap To Target Object Reference=@n4) -> (Snap Grid Line Thickness:int=0))
   ) -> (then)
   SetQuadPoints(execute=@n3.else, self=Get Grid Quad Component(), A="0, 0, 0", B="0, 0, 0", C="0, 0, 0", D="0, 0, 0")
  ]}
 On Canvas Snap Side Update {nodes [
   n1 = On Canvas Snap Side Update!() -> (then, Canvas:Canvas Object Reference, Width:int, Height:int) {is=title K2Node_FunctionEntry}
   > Draw Texture!(
    execute
    self:Canvas Object Reference=@n1
    RenderTexture:Texture Object Reference=/Game/Characters/Player/Female/Medium/Bodies/F_MED_Commando_01/Textures/SolidWhite_MaskCompression
    ScreenPosition_X:double=0
    ScreenPosition_Y:double=0
    ScreenSize:Vector 2D Structure=MakeVector2D(X="To Float (Integer)"(InInt=@n1.Width), Y="To Float (Integer)"(InInt=@n1.Height))
    CoordinatePosition:Vector 2D Structure=(X=0,Y=0)
    CoordinateSize:Vector 2D Structure=(X=1,Y=1)
    RenderColor:Linear Color Structure=(R=1,G=1,B=1,A=1)
    BlendMode:EBlendMode Enum=BLEND_Translucent
    Rotation:float=0
    PivotPoint:Vector 2D Structure=(X=0.5,Y=0.5)
   ) -> (then)
  ]}
 On Snap To Hidden Changed Bool {nodes [
   On Snap To Hidden Changed Bool!() -> (then, PropertySet:Scriptable Interactive Tool Property Set Object Reference, PropertyName:String, bNewValue:bool=~) {is=title K2Node_FunctionEntry}
   > n1 = GetActorsToIgnore()
   > Set Current Actors To Ignore In Trace(Current Actors To Ignore In Trace:Array of Actor Object References=@n1) -> (Output_Get:Array of Actor Object References)
  ]}
 On Vertical Offset Changed {nodes [
   n1 = On Vertical Offset Changed!() -> (then, PropertySet:Scriptable Interactive Tool Property Set Object Reference, PropertyName:String, NewValue:double=~) {is=title K2Node_FunctionEntry}
   n2 = Set Previous Vertical Offset(execute=[@n1.then, @n5.then], Previous Vertical Offset:double=Get Current Vertical Offset() -> (Current Vertical Offset:double)) -> (Output_Get:double)
   > n3 = For Each Loop(Array=Get ActorsToMove())
   Set Current Vertical Offset(execute=@n3.Completed, Current Vertical Offset:double=@n1.NewValue) -> (Output_Get:double)
   > Set MoveCommitted(MoveCommitted=true)
   n4 = float - float(A=@n1.NewValue, B=@n2)
   Set Actor Location(
    execute=@n3.LoopBody
    self=@n3
    NewLocation=vector + vector(A=MakeVector(Y=0, Z=@n4), B=Get Actor Location(self=@n3))
    bSweep=false
    bTeleport=false
   )
   n5 = Branch(Condition=Get FirstInitRun())
  ]}
 OnBeginHover {nodes [
   n1 = OnBeginHover!() -> (
    then
    CurrentPos_WorldRay:Ray Structure
    CurrentPos_bHas2D:bool=~
    CurrentPos_ScreenPosition:Vector 2D Structure
    Modifiers:Scriptable Tool Modifier States Structure
   ) {is=title K2Node_FunctionEntry}
   > Branch(Condition=Get PreviewKeyDown())
   > ActiveHover!(execute, self, HoverCurPosWorldRay:Ray Structure=@n1) -> (then) {is=title K2Node_CallFunction}
  ]}
 OnEventDuplicateSelected {nodes [
   OnEventDuplicateSelected() {is=title K2Node_FunctionEntry}
   > BeginTransaction(
    Context=Get TransactionContext() -> (TransactionContext:String)
    Description=Get Duplicate Transaction Description() -> (Duplicate Transaction Description:Text)
   )
   > TransactObject(Object=Self-Reference())
   > n1 = GetSelectedLevelActors(self=Get EditorActorSubsystem())
   > n2 = For Each Loop(Array=@n1)
   n3 = DuplicateActors!(
    execute=@n2.Completed
    self:Editor Actor Subsystem Object Reference=Get EditorActorSubsystem()
    ActorsToDuplicate:Array of Actor Object References=@n1
    ToWorld:World Object Reference
    Offset:Vector
   ) -> (then, ReturnValue:Array of Actor Object References)
   > Append Array(TargetArray=Get ActorsDuplicatedFromSnapTool(), SourceArray=@n3)
   > n4 = Set ActorsToMove(ActorsToMove:Array of Actor Object References=@n3) -> (Output_Get:Array of Actor Object References)
   > n5 = GetActorsToIgnore()
   > Set Current Actors To Ignore In Trace(Current Actors To Ignore In Trace=@n5)
   > SetSelectedLevelActors(self=Get EditorActorSubsystem(), ActorsToSelect=@n4)
   n6 = Get ActorsToMove()
   > n7 = GetActorArrayBounds(Actors=@n6, bOnlyCollidingComponents=false)
   n8 = Get Settings()
   > Set Offset Amount(Offset Amount:double=float * float(
     A=GetMaxElement!(A:Vector=@n7.BoxExtent) -> (ReturnValue:double)
     B=Get Duplicated Offset Amount!(self:BP STPS Snap To Target Object Reference=@n8) -> (Duplicated Offset Amount:double)
    )) -> (Output_Get:double)
   > n9 = For Each Loop(Array=@n6)
   n10 = Branch(
    execute=@n9.Completed
    Condition=Get Snap Each Object Along Axis!(self:BP STPS Snap To Target Object Reference=Get Settings()) -> (Snap Each Object Along Axis:bool)
   )
   > n11 = PlaceActorsOnGround()
   EndTransaction(execute=[@n11.then, @n10.else])
   n12 = BreakTransform(InTransform="Get (a copy)"(Array:Array of Transforms=Get ActorTransforms() -> (ActorTransforms:Array of Transforms), Dimension 1=@n9.Array Index) -> (Output:Transform))
   n13 = Get Offset Amount() -> (Offset Amount:double)
   Set Actor Transform!(
    execute=@n9.LoopBody
    self:Actor Object Reference=@n9
    NewTransform:"Transform (by ref)"=MakeTransform(Location=vector + vector(A=@n12, B:double=@n13), Rotation=@n12.Rotation, Scale=@n12.Scale)
    bSweep:bool=false
    bTeleport:bool=false
   ) -> (then, SweepHitResult:Hit Result Structure, ReturnValue:bool) {is=title K2Node_CallFunction}
   TransactObject(execute=@n2.LoopBody, Object=@n2)
   > Add(TargetArray:Array of Transforms=Get ActorTransforms(), NewItem:"Transform (by ref)"=Get Actor Transform(self=@n2)) {is=title K2Node_CallArrayFunction}
   Get Duplicate Offset() -> (Duplicate Offset:Vector)
   Get Duplicate Offset Bound Percent() -> (Duplicate Offset Bound Percent:double)
  ]}
 OnHitByClick {nodes [
   OnHitByClick!() -> (
    then
    ClickPos:Input Device Ray Structure
    Modifiers_bShiftDown:bool=~
    Modifiers_bCtrlDown:bool=~
    Modifiers_bAltDown:bool=~
    MouseButton:EScriptableToolMouseButton Enum=LeftButton
   ) {is=title K2Node_FunctionEntry}
   > OnEventDuplicateSelected() {is=title K2Node_CallFunction}
  ]}
 OnHotkeyInterrupted {nodes [
   OnHotkeyInterrupted() {is=title K2Node_FunctionEntry}
   > n1 = Sequence!(execute) -> (then_0, then_1)
   > Branch(Condition=Get PreviewKeyDown())
   > OnPreviewKeyReleased!(execute, self, Key:Key Structure=None, Modifiers:Scriptable Tool Modifier States Structure) -> (then)
   Branch(execute=@n1.then_1, Condition=Get MoveVertKeyDown())
   > OnVertKeyReleased!(execute, self, Key:Key Structure=None, Modifiers:Scriptable Tool Modifier States Structure) -> (then)
  ]}
 OnMoveAlongAxisKeyPressed {nodes [
   OnMoveAlongAxisKeyPressed!() -> (then, Key:Key Structure=None, Modifiers:Scriptable Tool Modifier States Structure) {is=title K2Node_FunctionEntry}
   > n1 = Branch(Condition=Get PreviewKeyDown())
   n2 = Branch(execute=@n1.else, Condition=Get MoveVertKeyDown())
   BeginTransaction(execute=@n2.else, Context=Get TransactionContext(), Description=Get Snap Transaction Description() -> (Snap Transaction Description:Text))
   > TransactObject(Object=Self-Reference())
   > Set Move Along Axis Key Down(Move Along Axis Key Down:bool=true) -> (Output_Get:bool)
   > Set Actors To Move() {is=title K2Node_CallFunction}
  ]}
 OnMoveAlongAxisKeyReleased {nodes [
   OnMoveAlongAxisKeyReleased!() -> (then, Key:Key Structure=None, Modifiers:Scriptable Tool Modifier States Structure) {is=title K2Node_FunctionEntry}
   > n1 = Branch(Condition=Get PreviewKeyDown())
   EndTransaction(execute=@n1.else)
   > Set Move Along Axis Key Down(Move Along Axis Key Down=false)
  ]}
 OnPivotAdjustKeyPressed {nodes [
   OnPivotAdjustKeyPressed!() -> (then, Key:Key Structure=None, Modifiers:Scriptable Tool Modifier States Structure) {is=title K2Node_FunctionEntry}
   > Set Pivot Adjust Key Down(Pivot Adjust Key Down:bool=true) -> (Output_Get:bool)
   n1 = Get PreviewKeyDown()
   n2 = Get Move Along Axis Key Down()
   n3 = Get MoveVertKeyDown()
   > Branch(Condition=AND Boolean!(A=NOT Boolean(A=@n1), B=NOT Boolean(A=@n2), C:bool=NOT Boolean(A=@n3)) -> (ReturnValue))
   > n4 = GetActorArrayBounds(Actors=Get ActorsToMove(), bOnlyCollidingComponents=false)
   > Set ActorsCenterStartWidgetXform(ActorsCenterStartWidgetXform:Vector=@n4) -> (Output_Get:Vector)
   > AddPivotAdjustWidget() {is=title K2Node_CallFunction}
  ]}
 OnPivotAdjustKeyReleased {nodes [
   OnPivotAdjustKeyReleased!() -> (then, Key:Key Structure=None, Modifiers:Scriptable Tool Modifier States Structure) {is=title K2Node_FunctionEntry}
   > Set Pivot Adjust Key Down(Pivot Adjust Key Down=false)
   > DestroyTRSGizmo!(execute, self:Scriptable Interactive Tool Object Reference, Identifier:String=Get PivotGizmoIdentifier()) -> (Success:Exec, Failure:Exec)
  ]}
 OnPreviewKeyPressed {nodes [
   OnPreviewKeyPressed!() -> (then, Key:Key Structure=None, Modifiers:Scriptable Tool Modifier States Structure) {is=title K2Node_FunctionEntry}
   > n1 = Branch(Condition=Get MoveVertKeyDown())
   n2 = Branch(execute=@n1.else, Condition=Get Move Along Axis Key Down())
   n3 = Branch(execute=@n2.else, Condition=Get PreviewKeyDown())
   Set MoveCommitted(execute=@n3.else, MoveCommitted=false)
   > n4 = Sequence!(execute) -> (then_0, then_1, then_2:Exec, then_3:Exec, then_4:Exec, then_5:Exec)
   n5 = Get Settings()
   > Set Side Text(Side Text:String=Enum to String!(Enumerator:E_SnapToTarget_BoundSide Enum=Get Selection Snap Axis(self=@n5) -> (Selection Snap Axis=NewEnumerator0)) -> (ReturnValue:String)) -> (Output_Get:String)
   Update Canvas Dynmats(execute=@n4.then_1)
   > UpdateResource(self=Get SnapSideText Render Target())
   > UpdateResource(self=Get Grid Draw Render Target())
   n6 = AddLine!(execute=@n4.then_2, self:Scriptable Tool Line Set Object Reference=Get Line Set() -> (Line Set:Scriptable Tool Line Set Object Reference)) -> (then, ReturnValue:Scriptable Tool Line Object Reference)
   > n7 = Set Dist Line(Dist Line:Scriptable Tool Line Object Reference=@n6) -> (Output_Get:Scriptable Tool Line Object Reference)
   > SetLineColor!(
    execute
    self:Scriptable Tool Line Object Reference=@n7
    Color:Color Structure=Make Color!(B:Byte=0, G:Byte=255, R:Byte=0, A:Byte=1) -> (Color:Color Structure) {is=title K2Node_MakeStruct}
   ) -> (then)
   > SetLineThickness!(execute, self:Scriptable Tool Line Object Reference=Get Dist Line(), Thickness:float=5) -> (then)
   n8 = AddLine(execute=@n4.then_3, self=Get Line Set())
   > n9 = Set RotPivotLine(RotPivotLine:Scriptable Tool Line Object Reference=@n8) -> (Output_Get:Scriptable Tool Line Object Reference)
   > SetLineColor(self=@n9, Color=Make Color(B=15, G=128, R=255, A=1))
   > SetLineThickness(self=Get RotPivotLine() -> (RotPivotLine:Scriptable Tool Line Object Reference), Thickness=5)
   BeginTransaction(execute=@n4.then_4, Context=Get TransactionContext(), Description=Get Snap Transaction Description())
   > TransactObject(Object=Self-Reference())
   > Set PreviewKeyDown(PreviewKeyDown:bool=true) -> (Output_Get:bool)
   > Set Actors To Move()
   n10 = GetActorsToIgnore(execute=@n4.then_5)
   > Set Current Actors To Ignore In Trace(Current Actors To Ignore In Trace=@n10)
  ]}
 OnPreviewKeyReleased {nodes [
   OnPreviewKeyReleased!() -> (then, Key:Key Structure=None, Modifiers:Scriptable Tool Modifier States Structure) {is=title K2Node_FunctionEntry}
   > n1 = Branch(Condition=Get MoveVertKeyDown())
   Set PreviewKeyDown(execute=@n1.else, PreviewKeyDown=false)
   > RemoveLine!(execute, self:Scriptable Tool Line Set Object Reference=Get Line Set(), Line:Scriptable Tool Line Object Reference=Get Dist Line()) -> (then)
   > RemoveLine(self=Get Line Set(), Line=Get RotPivotLine())
   > n2 = For Each Loop(Array:Array of Scriptable Tool Line Object References=Get Box Lines()) -> (Array Element:Scriptable Tool Line Object Reference)
   Set MoveCommitted(execute=@n2.Completed, MoveCommitted=true)
   > n3 = Sequence()
   > n4 = Branch(Condition=Get MoveCommitted() -> (MoveCommitted:bool))
   > Branch(Condition=Get Snap Each Object Along Axis(self=Get Settings()))
   > PlaceActorsOnGround()
   ResetActorsToMoveToPreMove(execute=@n4.else)
   n5 = Get Settings()
   Branch(execute=@n3.then_1, Condition=Get Show Snap Axis Plane(self=@n5))
   > SetQuadPoints(self=Get SnapSide_PlaneQuadComponent(), A="0, 0, 0", B="0, 0, 0", C="0, 0, 0", D="0, 0, 0")
   Branch(execute=@n3.then_2, Condition=Get Show Snap Grid(self=@n5))
   > SetQuadPoints(self=Get Grid Quad Component(), A="0, 0, 0", B="0, 0, 0", C="0, 0, 0", D="0, 0, 0")
   n6 = Set CurDistanceText(execute=@n3.then_3, CurDistanceText="LOCTEXT(\"0\")")
   > SetInfoText(DistanceValue=@n6, WheelText="LOCTEXT(\"0\")")
   Set Cur Rotation Amount(execute=@n3.then_5, Cur Rotation Amount=0)
   > EndTransaction()
   SetLineThickness(execute=@n2.LoopBody, self=@n2, Thickness=0)
  ]}
 OnSnapEachButtonClicked {nodes [
   OnSnapEachButtonClicked() {is=title K2Node_FunctionEntry}
   > BeginTransaction(Context=Get TransactionContext(), Description=Get SnapAlongAxisTransactionDesc() -> (SnapAlongAxisTransactionDesc:Text))
   > n1 = For Each Loop(Array:Array of Actor Object References=Get ActorsToMove()) -> (Array Element:Actor Object Reference)
   PlaceActorsOnGround(execute=@n1.Completed)
   > Set MoveCommitted(MoveCommitted=true)
   > EndTransaction()
   TransactObject(execute=@n1.LoopBody, Object=@n1)
  ]}
 OnSnapEachObjectHotkey {nodes [
   OnSnapEachObjectHotkey!() -> (then, Key:Key Structure=None, Modifiers:Scriptable Tool Modifier States Structure) {is=title K2Node_FunctionEntry}
   > OnSnapEachButtonClicked() {is=title K2Node_CallFunction}
  ]}
 OnSnapGridPropertiesChanged {nodes [
   OnSnapGridPropertiesChanged!() -> (then, PropertySet:Scriptable Interactive Tool Property Set Object Reference, PropertyName:String) {is=title K2Node_FunctionEntry}
   > Is Valid(InputObject=Get Settings())
   > n1 = Sequence()
   > Update Canvas Dynmats()
   n2 = For Each Loop(
    Exec=@n1.then_1
    Array:Array of Strings=Get PropertiesControlledByShowSnapGridRef() -> (PropertiesControlledByShowSnapGridRef:Array of Strings)
   ) -> (Array Element:String)
   n3 = Get Show Snap Grid(self=Get Settings())
   n4 = Get Grid Snapping(self=Get Settings())
   > SetPropertyAsReadOnly!(
    execute
    self:Scriptable Interactive Tool Property Set Object Reference=Get Settings()
    PropertyName:String=@n2
    bReadOnly:bool=OR Boolean(A=NOT Boolean(A=@n3), B=NOT Boolean(A=@n4))
   ) -> (then, ReturnValue:Scriptable Interactive Tool Property Set Object Reference)
   n5 = For Each Loop(Exec=@n1.then_2, Array=Get PropertiesControlledByGridSnapping() -> (PropertiesControlledByGridSnapping:Array of Strings))
   n6 = Get Settings()
   > SetPropertyAsReadOnly(self=Get Settings(), PropertyName=@n5, bReadOnly=NOT Boolean(A=Get Grid Snapping(self=@n6)))
   n7 = Switch on E_SnapToTarget_BoundSide(execute=@n1.then_4, Selection=Get Selection Snap Axis(self=Get Settings()) -> (Selection Snap Axis=NewEnumerator0))
   n8 = For Each Loop(Exec=@n7.NewEnumerator6, Array=Get PropertiesControlledByBoundsideCentre() -> (PropertiesControlledByBoundsideCentre:Array of Strings))
   SetIsEnabled(execute=@n8.Completed, self=Get Snap Button(), bInIsEnabled=false)
   SetPropertyAsReadOnly(execute=@n8.LoopBody, self=Get Settings(), PropertyName=@n8, bReadOnly=true)
   n9 = For Each Loop(
    Exec=[@n7.NewEnumerator0, @n7.NewEnumerator1, @n7.NewEnumerator2, @n7.NewEnumerator3, @n7.NewEnumerator4, @n7.NewEnumerator5]
    Array=Get PropertiesControlledByBoundsideCentre()
   )
   SetIsEnabled(execute=@n9.Completed, self=Get Snap Button(), bInIsEnabled=true)
   SetPropertyAsReadOnly(execute=@n9.LoopBody, self=Get Settings(), PropertyName=@n9, bReadOnly=false)
   n10 = Get Settings()
   SetPropertyAsReadOnly(execute=@n1.then_5, self=Get Settings(), PropertyName=Snap Axis Plane Color, bReadOnly=NOT Boolean(A=Get Show Snap Axis Plane(self=@n10)))
   n11 = For Each Loop(Exec=@n1.then_3, Array=Get PropertiesControlledByShowAxisPlane() -> (PropertiesControlledByShowAxisPlane:Array of Strings))
   n12 = Get Settings()
   > SetPropertyAsReadOnly(self=Get Settings(), PropertyName=@n11, bReadOnly=NOT Boolean(A=Get Show Snap Axis Plane(self=@n12)))
  ]}
 OnUpdateHover {nodes [
   n1 = OnUpdateHover!() -> (then, CurrentPos:Input Device Ray Structure, Modifiers:Scriptable Tool Modifier States Structure) {is=title K2Node_FunctionEntry}
   > Branch(Condition=Get PreviewKeyDown())
   > ActiveHover(HoverCurPosWorldRay=Break Input Device Ray!(InputDeviceRay:"Input Device Ray Structure (by ref)"=@n1) -> (WorldRay:Ray Structure, bHas2D:bool=~, ScreenPosition:Vector 2D Structure) {is=title K2Node_BreakStruct})
   > Return Node(ReturnValue=true)
  ]}
 OnVertKeyPressed {nodes [
   OnVertKeyPressed!() -> (then, Key:Key Structure=None, Modifiers:Scriptable Tool Modifier States Structure) {is=title K2Node_FunctionEntry}
   > n1 = Branch(Condition=Get PreviewKeyDown())
   n2 = Branch(execute=@n1.else, Condition=Get MoveVertKeyDown())
   BeginTransaction(execute=@n2.else, Context=Get TransactionContext(), Description=Get Snap Transaction Description())
   > TransactObject(Object=Self-Reference())
   > Set MoveVertKeyDown(MoveVertKeyDown:bool=true) -> (Output_Get:bool)
   > Set Actors To Move()
   > n3 = Set VertOffsetAtStartOfVertMove(VertOffsetAtStartOfVertMove:double=Get Vertical Offset(self=Get Settings())) -> (Output_Get:double)
   > Set Current Vertical Offset(Current Vertical Offset=@n3)
  ]}
 OnVertKeyReleased {nodes [
   OnVertKeyReleased!() -> (then, Key:Key Structure=None, Modifiers:Scriptable Tool Modifier States Structure) {is=title K2Node_FunctionEntry}
   > n1 = Branch(Condition=Get PreviewKeyDown())
   EndTransaction(execute=@n1.else)
   > Set MoveVertKeyDown(MoveVertKeyDown=false)
  ]}
 PivotGizmoTransformed {nodes [
   n1 = PivotGizmoTransformed!() -> (then, Gizmo Ident:String, New Widget Transform:Transform) {is=title K2Node_FunctionEntry}
   > Branch(Condition="Equal Exactly (String)"(A=@n1, B=Get PivotGizmoIdentifier()))
   Set Pivot Offset!(
    execute
    Pivot Offset:Vector=vector - vector(A=BreakTransform(InTransform=@n1.New Widget Transform), B=Get ActorsCenterStartWidgetXform() -> (ActorsCenterStartWidgetXform:Vector))
    self:BP STPS Snap To Target Object Reference=Get Settings()
   ) -> (then, Output_Get:Vector)
  ]}
 PlaceActorsOnGround {nodes [
   PlaceActorsOnGround() {is=title K2Node_FunctionEntry}
   > Set VertOffset(VertOffset:double=Get Vertical Offset(self=Get Settings())) -> (Output_Get:double)
   > n1 = For Each Loop(Array:Array of Actor Object References=Get ActorsToMove()) -> (Array Element:Actor Object Reference)
   Return Node!(execute=@n1.Completed)
   > n2 = Set CurActor(CurActor=@n1)
   n3 = Get Actor Bounds(self=@n2, bOnlyCollidingComponents=false)
   > n4 = GetBoxSideDirectionVector(BoundCenter=@n3, BoundExtents=@n3.BoxExtent)
   > n5 = Set RayStartPoint(RayStartPoint:Vector=@n4.Start Point) -> (Output_Get:Vector)
   > n6 = Set Ray Direction(Ray Direction:Vector=@n4) -> (Output_Get:Vector)
   > n7 = GetActorsToIgnore()
   n8 = GetRayStartEnd!(Ray_Origin:Vector=@n5, Ray_Direction:Vector=@n6, StartDistance=0, EndDistance=0) -> (StartPoint, EndPoint)
   > n9 = Line Trace By Channel(Start=@n8, End=@n8.EndPoint, bTraceComplex=true, ActorsToIgnore=@n7)
   > Branch(Condition=@n9.ReturnValue)
   n10 = Get CurActor()
   n11 = BreakVector(InVec=BreakHitResult!(Hit=@n9) -> (
     bBlockingHit
     bInitialOverlap
     Time
     Distance
     Location_X:double
     Location_Y:double
     Location_Z:double
     ImpactPoint
     Normal
     ImpactNormal
     PhysMat
     HitActor
     HitComponent
     HitBoneName
     BoneName
     HitItem
     ElementIndex
     FaceIndex
     TraceStart
     TraceEnd
    ).ImpactPoint)
   n12 = MakeVector(X=@n11, Y=@n11.Y, Z=float + float(A=@n11.Z, B=Get VertOffset() -> (VertOffset:double)))
   n13 = Get RayStartPoint() -> (RayStartPoint:Vector)
   > Set Actor Location(
    self=Get CurActor()
    NewLocation=vector - vector(A=Get Actor Location(self=@n10), B=vector - vector(A=@n13, B=@n12))
    bSweep=false
    bTeleport=false
   )
  ]}
 ResetActorsToMoveToPreMove {nodes [
   ResetActorsToMoveToPreMove() {is=title K2Node_FunctionEntry}
   > n1 = For Each Loop(Array=Get ActorsToMove())
   Return Node(execute=@n1.Completed)
   > Set Actor Location And Rotation(
    self=@n1
    NewLocation="Get (a ref)"(Array=Get PreMoveActorLocations() -> (PreMoveActorLocations:Array of Vectors), Dimension 1=@n1.Array Index)
    NewRotation="Get (a ref)"(Array:Array of Rotators=Get Pre Move Actor Rotations() -> (Pre Move Actor Rotations:Array of Rotators), Dimension 1=@n1.Array Index) -> (Output:"Rotator (by ref)")
    bSweep=false
    bTeleport=false
   )
  ]}
 ScriptSetup {nodes [
   ScriptSetup() {is=title K2Node_FunctionEntry}
   > n1 = Sequence!(execute) -> (then_0, then_1, then_2, then_3, then_4, then_5, then_6:Exec, then_7:Exec, then_8:Exec, then_9:Exec)
   > K2Node_Composite_5 = Setup Render Targets() {is=title K2Node_Composite}
   K2Node_Composite_0 = AddPropertySets(execute=@n1.then_2) {is=title K2Node_Composite}
   K2Node_Composite_3 = Add Behaviours!(InputPin:Exec=@n1.then_3) {is=title K2Node_Composite}
   K2Node_Composite_1 = Add Header Widget(execute=@n1.then_5) {is=title K2Node_Composite}
   K2Node_Composite_7 = AddConstVariables(execute=@n1.then_6) {is=title K2Node_Composite}
   Set Actors To Move(execute=@n1.then_7)
   K2Node_Composite_4 = Add Property Watches(execute=@n1.then_8) {is=title K2Node_Composite}
  ]}
 AddPropertySets {
  path=ScriptSetup.K2Node_Composite_0.AddPropertySets
  nodes [
   Inputs()
   > n1 = Sequence!(execute) -> (then_0, then_1)
   > n2 = AddPropertySetOfType(
    PropertySetType=&/BP_STPS_SnapToTarget.BP_STPS_SnapToTarget_C
    Identifier=Get SettingsIdentifier() -> (SettingsIdentifier:String)
   )
   > n3 = Cast To BP_STPS_SnapToTarget!(execute, Object=@n2) -> (then, CastFailed, AsBP STPS Snap to Target:BP STPS Snap To Target Object Reference) {is=title K2Node_DynamicCast}
   > n4 = Set Settings(Settings:BP STPS Snap To Target Object Reference=@n3) -> (Output_Get:BP STPS Snap To Target Object Reference)
   > RestorePropertySetSettings(PropertySet=@n4, SaveKey=Get PropSetSaveKey() -> (PropSetSaveKey:String))
   n5 = AddPropertySetOfType(
    execute=@n1.then_1
    PropertySetType=/Game/Blueprints/Blutilities/ToolsWrkshp/ScrTools/Shared/SharedPropertySets/BP_STPS_SharedWidgetState.BP_STPS_SharedWidgetState_C
    Identifier=Get SharedStateIdentifier() -> (SharedStateIdentifier:String)
   )
   > n6 = Cast To BP_STPS_SharedWidgetState!(execute, Object=@n5) -> (then, CastFailed, AsBP STPS Shared Widget State:BP STPS Shared Widget State Object Reference) {is=title K2Node_DynamicCast}
   > Set SharedState Settings(SharedState Settings:BP STPS Shared Widget State Object Reference=@n6) -> (Output_Get:BP STPS Shared Widget State Object Reference)
   > SetPropertySetVisibleByName!(execute, self:Scriptable Interactive Tool Object Reference, Identifier:String=Get SharedStateIdentifier(), bVisible:bool=false) -> (then)
   > RestorePropertySetSettings(PropertySet=Get SharedState Settings(), SaveKey=Get PropSetSaveKey())
   Outputs()
  ]
 }
 Add Header Widget {
  path=ScriptSetup.K2Node_Composite_1.Add Header Widget
  nodes [
   Inputs()
   > Sequence()
   > n1 = Construct Object from Class(Class=&BP_UW_SharedHeaderWidget_C) -> (ReturnValue:BP UW Shared Header Widget Object Reference)
   > n2 = Set Header Widget(Header Widget:BP UW Shared Header Widget Object Reference=@n1) -> (Output_Get:BP UW Shared Header Widget Object Reference)
   > SetToolPanelToolHeaderWidget(Widget=@n2)
   Outputs()
  ]
 }
 Add Behaviours {
  path=ScriptSetup.K2Node_Composite_3.Add Behaviours
  nodes [
   n1 = Create Event()
   n2 = AddSingleKeyInputBehavior!(
    execute
    self:Scriptable Modular Behavior Tool Object Reference
    OnKeyPressed:Delegate=Create Event()
    OnKeyReleased:Delegate=Create Event()
    OnForceEndCaptureFuncIn:Delegate=@n1
    Key:Key Structure=Get MoveAlongAxisKey() -> (MoveAlongAxisKey:Key Structure)
    CaptureCheck:Delegate
    CapturePriority:int=0
   ) -> (then)
   AddSingleKeyInputBehavior(
    execute=[@n2.then, @n5.then]
    OnKeyPressed=Create Event()
    OnForceEndCaptureFuncIn=@n1
    Key=Get SnapEachObjectKey() -> (SnapEachObjectKey:Key Structure)
    CapturePriority=0
   )
   Inputs!() -> (InputPin:Exec)
   > n3 = Sequence()
   AddMouseWheelBehavior!(
    execute=@n3.then_1
    self:Scriptable Modular Behavior Tool Object Reference
    TestShouldRespondToMouseWheel:Delegate=Create Event()
    OnMouseWheelScrollUp:Delegate=Create Event()
    OnMouseWheelScrollDown:Delegate=Create Event()
    CaptureCheck:Delegate=Create Event()
    CapturePriority:int=0
   ) -> (then)
   n4 = Create Event()
   AddSingleClickBehavior!(
    execute=@n3.then_0
    self:Scriptable Modular Behavior Tool Object Reference
    IfHitByClick:Delegate=Create Event()
    OnHitByClick:Delegate=Create Event()
    CaptureCheck:Delegate=@n4
    CapturePriority:int=100
    MouseButton:EScriptableToolMouseButton Enum=LeftButton
    bHitTestOnRelease:bool=true
   ) -> (then)
   > AddMouseHoverBehavior!(
    execute
    self:Scriptable Modular Behavior Tool Object Reference
    BeginHoverSequenceHitTest:Delegate=Create Event()
    OnBeginHover:Delegate=Create Event()
    OnUpdateHover:Delegate=Create Event()
    OnEndHover:Delegate
    HoverCaptureCheck:Delegate=@n4
    CapturePriority:int=100
   ) -> (then)
   > AddSingleKeyInputBehavior(
    OnKeyPressed=Create Event()
    OnKeyReleased=Create Event()
    OnForceEndCaptureFuncIn=@n1
    Key=Get PreviewKey() -> (PreviewKey:Key Structure)
    CapturePriority=0
   )
   > n5 = AddSingleKeyInputBehavior(
    OnKeyPressed=Create Event()
    OnKeyReleased=Create Event()
    OnForceEndCaptureFuncIn=@n1
    Key=Get MoveVerticalKey() -> (MoveVerticalKey:Key Structure)
    CapturePriority=0
   )
   AddSingleKeyInputBehavior(
    OnKeyPressed=Create Event()
    OnKeyReleased=Create Event()
    Key=Get PivotAdjustmentKey() -> (PivotAdjustmentKey:Key Structure)
    CapturePriority=0
   )
   Outputs()
  ]
 }
 Add Property Watches {
  path=ScriptSetup.K2Node_Composite_4.Add Property Watches
  nodes [
   Inputs!() -> (execute:Exec)
   > n1 = Sequence!(execute) -> (then_0, then_1, then_2:Exec)
   > n2 = For Each Loop(Array:Array of Strings=Get GridPropertiesToWatch() -> (GridPropertiesToWatch:Array of Strings)) -> (Array Element:String)
   > WatchProperty(PropertySet=Get Settings(), PropertyName=@n2, OnModified=Create Event())
   n3 = WatchFloatProperty!(
    execute=@n1.then_1
    self:Scriptable Interactive Tool Object Reference
    PropertySet:Scriptable Interactive Tool Property Set Object Reference=Get Settings()
    PropertyName:String=Vertical Offset
    OnModified:"Delegate (by ref)"=Create Event()
   ) -> (then, ReturnValue:Scriptable Interactive Tool Property Set Object Reference)
   > WatchBoolProperty!(
    execute
    self:Scriptable Interactive Tool Object Reference
    PropertySet:Scriptable Interactive Tool Property Set Object Reference=@n3
    PropertyName:String=Snap To Hidden
    OnModified:"Delegate (by ref)"=Create Event()
   ) -> (then, ReturnValue:Scriptable Interactive Tool Property Set Object Reference)
   Outputs()
  ]
 }
 Setup Render Targets {
  path=ScriptSetup.K2Node_Composite_5.Setup Render Targets
  nodes [
   Inputs()
   > n1 = Sequence!(execute) -> (then_0, then_1)
   n2 = Get Text Canvas Resolution() -> (Text Canvas Resolution:int=0)
   > n3 = CreateCanvasRenderTarget2D!(
    execute
    CanvasRenderTarget2DClass:Canvas Render Target 2D Class Reference=/Script/Engine.CanvasRenderTarget2D
    Width:int=@n2
    Height:int=@n2
   ) -> (then, ReturnValue:Canvas Render Target 2D Object Reference)
   > n4 = Set SnapSideText Render Target(SnapSideText Render Target:Canvas Render Target 2D Object Reference=@n3) -> (Output_Get:Canvas Render Target 2D Object Reference)
   > Bind Event to On Canvas Render Target Update(self:Canvas Render Target 2D Object Reference=@n4, Delegate=Create Event()) {is=title K2Node_AddDelegate}
   > UpdateResource(self=Get SnapSideText Render Target())
   n5 = Get Grid Canvas Resolution() -> (Grid Canvas Resolution:int=0)
   n6 = CreateCanvasRenderTarget2D(execute=@n1.then_1, CanvasRenderTarget2DClass=&CanvasRenderTarget2D, Width=@n5, Height=@n5)
   > n7 = Set Grid Draw Render Target(Grid Draw Render Target:Canvas Render Target 2D Object Reference=@n6) -> (Output_Get:Canvas Render Target 2D Object Reference)
   > Bind Event to On Canvas Render Target Update(self=@n7, Delegate=Create Event())
   > UpdateResource(self=Get Grid Draw Render Target())
   Outputs()
  ]
 }
 AddConstVariables {
  path=ScriptSetup.K2Node_Composite_7.AddConstVariables
  nodes [
   Inputs()
   > n1 = Sequence!(execute) -> (then_0, then_1, then_2:Exec, then_3:Exec, then_4:Exec, then_5:Exec, then_6:Exec, then_7:Exec, then_8:Exec)
   > n2 = AddLineSet!(execute, self:Scriptable Interactive Tool Object Reference) -> (then, ReturnValue:Scriptable Tool Line Set Object Reference)
   > Set Line Set(Line Set:Scriptable Tool Line Set Object Reference=@n2) -> (Output_Get:Scriptable Tool Line Set Object Reference)
   For Loop(execute=@n1.then_1, LastIndex=11)
   > n3 = AddLine(self=Get Line Set())
   > Add(TargetArray:Array of Scriptable Tool Line Object References=Get Box Lines(), NewItem:"Scriptable Tool Line Object Reference (by ref)"=@n3)
   n4 = AddTriangleSet!(execute=@n1.then_2, self:Scriptable Interactive Tool Object Reference) -> (then, ReturnValue:Scriptable Tool Triangle Set Object Reference)
   > Set PlaneTriSet(PlaneTriSet:Scriptable Tool Triangle Set Object Reference=@n4) -> (Output_Get:Scriptable Tool Triangle Set Object Reference)
   n5 = AddQuad!(
    execute=@n1.then_3
    self:Scriptable Tool Triangle Set Object Reference=Get PlaneTriSet() -> (PlaneTriSet:Scriptable Tool Triangle Set Object Reference)
   ) -> (then, ReturnValue:Scriptable Tool Quad Object Reference)
   > Set SnapSide_PlaneQuadComponent(SnapSide_PlaneQuadComponent:Scriptable Tool Quad Object Reference=@n5) -> (Output_Get:Scriptable Tool Quad Object Reference)
   n6 = AddQuad(execute=@n1.then_4, self=Get PlaneTriSet())
   > Set Grid Quad Component(Grid Quad Component:Scriptable Tool Quad Object Reference=@n6) -> (Output_Get:Scriptable Tool Quad Object Reference)
   n7 = CreateDynamicMaterialInstance!(
    execute=@n1.then_5
    Parent=Get BoxSidePlaneNoDepthTest_M() -> (BoxSidePlaneNoDepthTest_M:Material Object Reference)
    OptionalName=GridMaterial
    CreationFlags=Transient
   ) -> (then, ReturnValue)
   > Set GridMat_NoDepthTest_DynMat(GridMat_NoDepthTest_DynMat:Material Instance Dynamic Object Reference=@n7) -> (Output_Get:Material Instance Dynamic Object Reference)
   > n8 = CreateDynamicMaterialInstance(
    Parent=Get BoxSidePlaneOccluded_M() -> (BoxSidePlaneOccluded_M:Material Object Reference)
    OptionalName=GridMaterial
    CreationFlags=Transient
   )
   > Set Grid Mat Occluded Dyn Mat(Grid Mat Occluded Dyn Mat:Material Instance Dynamic Object Reference=@n8) -> (Output_Get:Material Instance Dynamic Object Reference)
   Update Canvas Dynmats(execute=@n1.then_6)
   n9 = "To Color (LinearColor)"!(
    InLinearColor:Linear Color Structure=Get Snap Axis Plane Color!(self:BP STPS Snap To Target Object Reference=Get Settings()) -> (Snap Axis Plane Color:Linear Color Structure)
    InUseSRGB:bool=true
   ) -> (ReturnValue:Color Structure)
   SetQuadColors!(
    execute=@n1.then_7
    self:Scriptable Tool Quad Object Reference=Get SnapSide_PlaneQuadComponent()
    A:Color Structure=@n9
    B:Color Structure=@n9
    C:Color Structure=@n9
    D:Color Structure=@n9
   ) -> (then)
   > SetQuadUVs!(
    execute
    self:Scriptable Tool Quad Object Reference=Get SnapSide_PlaneQuadComponent()
    A:Vector 2D Structure
    B:Vector 2D Structure=(X=0,Y=1)
    C:Vector 2D Structure=(X=1,Y=1)
    D:Vector 2D Structure=(X=1,Y=0)
   ) -> (then)
   n10 = "To Color (LinearColor)"(InLinearColor=Get Snap Grid Color(self=Get Settings()), InUseSRGB=true)
   SetQuadColors(execute=@n1.then_8, self=Get Grid Quad Component(), A=@n10, B=@n10, C=@n10, D=@n10)
   > SetQuadUVs(self=Get Grid Quad Component(), B=(X=0,Y=1), C=(X=1,Y=1), D=(X=1,Y=0))
   > SetQuadNormals(self=Get Grid Quad Component(), A="0,0,1", B="0,0,1", C="0,0,1", D="0,0,1")
   Outputs()
  ]
 }
 ScriptShutdown {nodes [
   ScriptShutdown!() -> (then, ShutdownType:EToolShutdownType Enum) {is=title K2Node_FunctionEntry}
   > FindCoincidentActors!(execute, self, ActorsToTest:Array of Actor Object References=Get ActorsToMove()) -> (then) {is=title K2Node_CallFunction}
   > n1 = Branch(Condition=Get MoveCommitted())
   Set Vertical Offset(execute=[@n4.then, @n1.then], Vertical Offset=0, self=Get Settings())
   > SavePropertySetSettings(PropertySet=Get Settings(), SaveKey=Get PropSetSaveKey())
   n2 = Get Readout Box()
   > Set ReadoutOpenOnExit!(
    execute
    ReadoutOpenOnExit:bool=IsChecked(self=Get ExpandButton!(self:BP UW Expandable Box Object Reference=@n2) -> (ExpandButton:Check Box Object Reference))
    self:BP STPS Shared Widget State Object Reference=Get SharedState Settings()
   ) -> (then, Output_Get:bool)
   n3 = Get Shortcut Box()
   > Set ShortcutsOpenOnExit!(
    execute
    ShortcutsOpenOnExit:bool=IsChecked(self=Get ExpandButton(self=@n3))
    self:BP STPS Shared Widget State Object Reference=Get SharedState Settings()
   ) -> (then, Output_Get:bool)
   > SavePropertySetSettings(PropertySet=Get SharedState Settings(), SaveKey=Get PropSetSaveKey())
   > EndTransaction()
   > RemoveAllLines(self:Scriptable Tool Line Set Object Reference=Get Line Set())
   > Set Line Set()
   n4 = ResetActorsToMoveToPreMove(execute=@n1.else) {is=title K2Node_CallFunction}
  ]}
 Set Actors To Move {nodes [
   Set Actors To Move() {is=title K2Node_FunctionEntry}
   > Clear(TargetArray:Array of Vectors=Get PreMoveActorLocationsWithRotOffset())
   > Clear(TargetArray=Get PreMoveActorLocations())
   > Clear(TargetArray:Array of Actor Object References=Get ActorsToMove())
   > n1 = GetSelectedLevelActors(self=Get EditorActorSubsystem())
   > Set Cur Rotation Amount(Cur Rotation Amount=0)
   > Set CurWheelText(CurWheelText="LOCTEXT(\"0\")")
   > n2 = Set ActorsToMove(ActorsToMove=@n1)
   > n3 = GetActorArrayBounds(Actors=@n2, bOnlyCollidingComponents=Get Only Colliding Components())
   > Set PreMoveBBoxCenter(PreMoveBBoxCenter:Vector=@n3) -> (Output_Get:Vector)
   > Set PreMoveBBoxExtent(PreMoveBBoxExtent:Vector=@n3.BoxExtent) -> (Output_Get:Vector)
   > n4 = For Each Loop(Array:Array of Actor Object References=Get ActorsToMove()) -> (Array Element:Actor Object Reference)
   Set PreMoveActorLocationsWithRotOffset(execute=@n4.Completed, PreMoveActorLocationsWithRotOffset:Array of Vectors=Get PreMoveActorLocations()) -> (Output_Get:Array of Vectors)
   > Append Array(TargetArray=Get ActorsToIgnoreInTrace(), SourceArray=Get ActorsToMove())
   TransactObject(execute=@n4.LoopBody, Object=@n4)
   > Add(TargetArray:Array of Vectors=Get PreMoveActorLocations(), NewItem:"Vector (by ref)"=Get Actor Location(self=@n4))
   > Add(TargetArray:Array of Rotators=Get Pre Move Actor Rotations(), NewItem:"Rotator (by ref)"=Get Actor Rotation(self=@n4))
  ]}
 SetInfoText {nodes [
   n1 = SetInfoText!() -> (then, DistanceValue:Text="LOCTEXT(\"0\")", WheelText:Text="LOCTEXT(\"0\")") {is=title K2Node_FunctionEntry}
   > SetLine!(
    execute
    self:BP UW Readout Line Object Reference=Get Distance Line() -> (Distance Line:BP UW Readout Line Object Reference)
    Label:Text="LOCTEXT(\"Distance (cm):\")"
    Value:Text=@n1
   ) -> (then)
   > SetLine(self=Get Angle Line() -> (Angle Line:BP UW Readout Line Object Reference), Label="LOCTEXT(\"Angle Rotated (deg):\")", Value=@n1.WheelText)
   > Return Node()
  ]}
 Should Respond To Mouse Wheel {nodes [
   Should Respond To Mouse Wheel!() -> (then, CurrentPos:Input Device Ray Structure) {is=title K2Node_FunctionEntry}
   n1 = Get PreviewKeyDown()
   n2 = Get MoveVertKeyDown()
   > Return Node!(
    execute
    ReturnValue_bHit:bool=XOR Boolean(A=XOR Boolean(A=@n1, B=@n2), B=Get Move Along Axis Key Down())
    ReturnValue_HitDepth:double=340282346638528859811704183484516925440
    ReturnValue_HitNormal:Vector="0,0,1"
    ReturnValue_bHasHitNormal:bool=False
    ReturnValue_HitIdentifier:int=0
    ReturnValue_HitObject:Object Reference
   )
  ]}
 SingleClickCaptureCheck {nodes [
   SingleClickCaptureCheck!() -> (
    then
    InputDeviceState_bShiftKeyDown:bool=~
    InputDeviceState_bAltKeyDown:bool=~
    InputDeviceState_bCtrlKeyDown:bool=~
    InputDeviceState_bCmdKeyDown:bool=~
    InputDeviceState_Keyboard_ActiveKey_Button:Key Structure
    InputDeviceState_Keyboard_ActiveKey_bPressed:bool=~
    InputDeviceState_Keyboard_ActiveKey_bDown:bool=~
    InputDeviceState_Keyboard_ActiveKey_bReleased:bool=~
    InputDeviceState_Keyboard_ActiveKey_bDoubleClicked:bool=~
    InputDeviceState_Mouse_Left_Button:Key Structure
    InputDeviceState_Mouse_Left_bPressed:bool=~
    InputDeviceState_Mouse_Left_bDown:bool=~
    InputDeviceState_Mouse_Left_bReleased:bool=~
    InputDeviceState_Mouse_Left_bDoubleClicked:bool=~
    InputDeviceState_Mouse_Middle:Device Button State Structure
    InputDeviceState_Mouse_Right:Device Button State Structure
    InputDeviceState_Mouse_WheelDelta:float=~
    InputDeviceState_Mouse_Position2D:Vector 2D Structure
    InputDeviceState_Mouse_Delta2D:Vector 2D Structure
    InputDeviceState_Mouse_WorldRay:Ray Structure
    InputDeviceState_Mouse_UnboundedPosition2D:Vector 2D Structure
    InputDeviceState_Mouse_UnboundedMouseDelta2D:Vector 2D Structure
    InputDeviceState_Mouse_UnboundedWorldRay:Ray Structure
   ) {is=title K2Node_FunctionEntry}
   > SetFocusInViewport(self:Scriptable Interactive Tool Object Reference)
   > Return Node!(execute, ReturnValue:bool=Get PreviewKeyDown())
  ]}
 TransformActorGroup {nodes [
   n1 = TransformActorGroup!() -> (then, NewLocation:Vector=~, GroupBoundCenter:Vector=~, GroupBoundExtent:Vector=~, Rotation:Rotator=~) {is=title K2Node_FunctionEntry}
   > n2 = Get Bounds Edge Center Locations(BoundCenter=Get PreMoveBBoxCenter(), BoundExtents=@n1.GroupBoundExtent)
   n3 = Get Settings()
   > Set BoundEdgeLocation(BoundEdgeLocation:Vector=Select(
     NewEnumerator0=@n2.MinY
     NewEnumerator1=@n2.MaxY
     NewEnumerator2=@n2
     NewEnumerator3=@n2.MinX
     NewEnumerator4=@n2.MaxZ
     NewEnumerator5=@n2.MinZ
     NewEnumerator6=Get PreMoveBBoxCenter()
     Index=Get Selection Snap Axis(self=@n3) -> (Selection Snap Axis=NewEnumerator0)
    )) -> (Output_Get:Vector)
   > n4 = For Each Loop(Array=Get ActorsToMove())
   n5 = "Get (a copy)"(Array:Array of Vectors=Get PreMoveActorLocationsWithRotOffset(), Dimension 1=@n4.Array Index) -> (Output:Vector)
   n6 = Get BoundEdgeLocation() -> (BoundEdgeLocation:Vector)
   > Set Actor Location(
    self=@n4
    NewLocation=vector + vector(A=vector - vector(A=@n5, B=@n6), B:Vector=Get NewLocation() -> (NewLocation:Vector))
    bSweep=false
    bTeleport=false
   )
  ]}
 Update Canvas Dynmats {nodes [
   Update Canvas Dynmats() {is=title K2Node_FunctionEntry}
   > n1 = Sequence!(execute) -> (then_0, then_1, then_2)
   > n2 = CreateDynamicMaterialInstance(
    Parent=Get BoxSidePlaneNoDepthTest_MI() -> (BoxSidePlaneNoDepthTest_MI:Material Instance Object Reference)
    OptionalName=SnapGridMat
    CreationFlags=Transient
   )
   > n3 = Set GridMat_NoDepthTest_DynMat(GridMat_NoDepthTest_DynMat=@n2)
   SetQuadMaterial!(
    execute=[@n9.else, @n3.then]
    self:Scriptable Tool Quad Object Reference=Get Grid Quad Component()
    Material:Material Interface Object Reference=@n3
   ) -> (then)
   > SetTextureParameterValue(
    self=Get GridMat_NoDepthTest_DynMat() -> (GridMat_NoDepthTest_DynMat:Material Instance Dynamic Object Reference)
    ParameterName=CanvasFront
    Value=Get Grid Draw Render Target()
   )
   n4 = "To Color (LinearColor)"(InLinearColor=Get Snap Grid Color(self=Get Settings()), InUseSRGB=true)
   > SetQuadColors(self=Get Grid Quad Component(), A=@n4, B=@n4, C=@n4, D=@n4)
   n5 = Get Settings()
   > SetScalarParameterValue(
    self=Get GridMat_NoDepthTest_DynMat()
    ParameterName=Do Radial Fade
    Value="To Float (Boolean)"(InBool=Get Radial Fade!(self:BP STPS Snap To Target Object Reference=@n5) -> (Radial Fade:bool))
   )
   n6 = CreateDynamicMaterialInstance(execute=@n1.then_1, Parent=Get BoxSidePlaneNoDepthTest_MI(), OptionalName=SnapSideMat, CreationFlags=Transient)
   > Set SnapSideMat_DynMat(SnapSideMat_DynMat:Material Instance Dynamic Object Reference=@n6) -> (Output_Get:Material Instance Dynamic Object Reference)
   n7 = "To Color (LinearColor)"(InLinearColor=Get Snap Axis Plane Color(self=Get Settings()), InUseSRGB=true)
   > SetQuadColors(self=Get SnapSide_PlaneQuadComponent(), A=@n7, B=@n7, C=@n7, D=@n7)
   n8 = Get SnapSideMat_DynMat() -> (SnapSideMat_DynMat:Material Instance Dynamic Object Reference)
   > SetQuadMaterial(self=Get SnapSide_PlaneQuadComponent(), Material=@n8)
   > SetTextureParameterValue(self=@n8, ParameterName=CanvasFront, Value=Get SnapSideText Render Target())
   > SetScalarParameterValue(self=Get SnapSideMat_DynMat(), ParameterName=Do Radial Fade, Value="To Float (Boolean)"(InBool=MakeLiteralBool(Value=false)))
   n9 = Branch(Condition=false)
   > SetQuadMaterial(
    self=Get Grid Quad Component()
    Material=Get Grid Mat Occluded Dyn Mat() -> (Grid Mat Occluded Dyn Mat:Material Instance Dynamic Object Reference)
   )
   > SetTextureParameterValue(self=Get Grid Mat Occluded Dyn Mat(), ParameterName=CanvasFront, Value=Get Grid Draw Render Target())
   n10 = Get Settings()
   > SetScalarParameterValue(self=Get Grid Mat Occluded Dyn Mat(), ParameterName=Do Radial Fade, Value="To Float (Boolean)"(InBool=Get Radial Fade(self=@n10)))
  ]}
 WaitForTime {nodes [
   n1 = WaitForTime!() -> (then, PauseTime:double=0.5) {is=title K2Node_FunctionEntry}
   n2 = Get LastWaitTime() -> (LastWaitTime:double)
   > n3 = Branch(Condition="float > float"(A=Get CurTime(), B=float + float(A=@n1, B=@n2)))
   > Set LastWaitTime(LastWaitTime:double=Get CurTime()) -> (Output_Get:double)
   > Return Node!(execute, Active:bool=true)
   Return Node(execute=@n3.else, Active=false)
  ]}
]
variables [
 Actions Box {type=BP_UW_ExpandableBox_C}
 ActiveHoverOutHit {
  type=HitResult
  default {
   bBlockingHit=false
   bStartPenetrating=false
   boneName=None
   component=null
   distance=0
   elementIndex=0
   faceIndex=0
   item=0
   myBoneName=None
   myItem=-1
   penetrationDepth=0
   physMaterial=null
   time=1
   hitObjectHandle {referenceObject=null}
   impactNormal=<0 0 0>
   impactPoint=<0 0 0>
   location=<0 0 0>
   normal=<0 0 0>
   traceEnd=<0 0 0>
   traceStart=<0 0 0>
  }
 }
 ActorsCenterStartWidgetXform {type=Vector}
 ActorsDuplicatedFromSnapTool {type="array<Actor>"}
 ActorsToIgnoreInTrace {type="array<Actor>"}
 ActorsToMove {type="array<Actor>"}
 Angle Line {type=BP_UW_ReadoutLine_C, default=null}
 Box Lines {type="array<ScriptableToolLine>", default []}
 BoxSidePlane_MI {type=MaterialInstance, default=/Game/Blueprints/Blutilities/ToolsWrkshp/ScrTools/Shared/Materials/MI_BoxSidePlane}
 BoxSidePlaneNoDepthTest_M {type=Material, default=/Game/Blueprints/Blutilities/ToolsWrkshp/ScrTools/Shared/Materials/M_BoxSidePlaneNoDepthTest, class=~, graph=~}
 BoxSidePlaneNoDepthTest_MI {type=MaterialInstance, default=/Game/Blueprints/Blutilities/ToolsWrkshp/ScrTools/Shared/Materials/MI_BoxSidePlaneNoDepthTest}
 BoxSidePlaneOccluded_M {type=Material, default=/Game/Blueprints/Blutilities/ToolsWrkshp/ScrTools/Shared/Materials/M_BoxSidePlane, class=~, graph=~}
 Cur Rotation Amount {type=number}
 CurDistanceText {type=string, default=""}
 Current Actors To Ignore In Trace {type="array<Actor>"}
 Current Line {type=BP_UW_ReadoutLine_C, default=null}
 Current Vertical Offset {type=number}
 CurTime {type=number}
 CurWheelText {type=string, default=0}
 Dist Line {type=ScriptableToolLine, default=null}
 Distance Line {type=BP_UW_ReadoutLine_C, default=null}
 DistConversionValues {
  type="map<number>"
  default {
   Centimeters=0.01
   Cubit=45.72
   "Football field (American)"=109.7
   "Football field (Canadian)"=100.6
   "Football field (aka Soccer)"=105
   Furlong=201.168
   Horse=2.4
   Light Nanosecond=29.9792
   Meters=1
   Smoot=1.7018
  }
 }
 DupButton {type=BP_UW_DupButton_C, default=null}
 Duplicate Offset {type=Vector, default=<25 25 25>}
 Duplicate Offset Bound Percent {type=number, default=0.02}
 Duplicate Transaction Description {type=string, default=Add Actor}
 EmptyOverlayText {type=string, default=""}
 FirstInitRun {type=boolean}
 Grid Canvas Resolution {type=integer, default=2048}
 Grid Draw Render Target {type=CanvasRenderTarget2D, default=null}
 Grid Mat Occluded Dyn Mat {type=MaterialInstanceDynamic}
 Grid Quad Component {type=ScriptableToolQuad, default=null}
 GridMat_NoDepthTest_DynMat {type=MaterialInstanceDynamic}
 HaveMovedAfterDuplicate {type=boolean, default=true}
 Header Widget {type=BP_UW_SharedHeaderWidget_C}
 HoverCurPosWorldRay {type=Ray, default {direction=<0 0 0>, origin=<0 0 0>}}
 HoveredSnappedLocation {type=Vector}
 LastWaitTime {type=number}
 LastWheelTime {type=number}
 Line Set {type=ScriptableToolLineSet, default=null}
 Move Along Axis Key Down {type=boolean}
 MoveAlongAxisKey {type=Key, default=V}
 MoveCommitted {type=boolean}
 MoveVerticalKey {type=Key, default=C}
 MoveVertKeyDown {type=boolean}
 Only Colliding Components {type=boolean, default=true}
 Pivot Adjust Key Down {type=boolean}
 PivotAdjustmentKey {type=Key, default=N}
 PivotGizmoIdentifier {type=string, default=PivotGizmo}
 PlaneTriSet {type=ScriptableToolTriangleSet, default=null}
 Pre Move Actor Rotations {type="array<Rotator>", default []}
 PreMoveActorLocations {type="array<Vector>"}
 PreMoveActorLocationsWithRotOffset {type="array<Vector>"}
 PreMoveBBoxCenter {type=Vector}
 PreMoveBBoxExtent {type=Vector}
 PreviewKey {type=Key, default=X}
 PreviewKeyDown {type=boolean}
 PreviewMoveIntervalTime {type=number, default=0.001}
 PropSetSaveKey {type=string, default=SnapToTargetSaveKey}
 Readout Box {type=BP_UW_ExpandableBox_C}
 RotPivotLine {type=ScriptableToolLine, default=null}
 Settings {type=BP_STPS_SnapToTarget_C, default=null}
 SettingsIdentifier {type=string, default=Settings}
 SharedState Settings {type=BP_STPS_SharedWidgetState_C, default=null}
 SharedStateIdentifier {type=string, default=Shared State Identifier}
 Shortcut Box {type=BP_UW_ExpandableBox_C}
 Shortcut Map {
  type="map<string>"
  default {
   Duplicate=Hold X + Left Click
   Move Up/Down=Hold C + Mouse Wheel
   Place=Release X
   Rotate=Hold X + Mouse Wheel
   Snap=Hold X
   Snap Each Object=Press B
  }
 }
 Shortcut Map Keys {type="array<string>", default [Snap, Place, Rotate, Duplicate, Move Up/Down, Snap Each Object]}
 ShortcutLines {type="array<BP_UW_ReadoutLine_C>", default []}
 Side Render Font {type=Font, default=&NotoSans}
 Side Text {type=string, default=""}
 Snap Button {type=BP_UW_SnapEachObjectButton_C, default=null}
 Snap Transaction Description {type=string, default=Snap Actors}
 SnapAlongAxisTransactionDesc {type=string, default=Actor Movement}
 SnapEachObjectKey {type=Key, default=B}
 SnapSide_PlaneQuadComponent {type=ScriptableToolQuad, default=null}
 SnapSideMat_DynMat {type=MaterialInstanceDynamic}
 SnapSideText Render Target {type=CanvasRenderTarget2D, default=null}
 Text Canvas Resolution {type=integer, default=1024}
 TraceOutHit {
  type=HitResult
  default {
   bBlockingHit=false
   bStartPenetrating=false
   boneName=None
   component=null
   distance=0
   elementIndex=0
   faceIndex=0
   item=0
   myBoneName=None
   myItem=-1
   penetrationDepth=0
   physMaterial=null
   time=1
   hitObjectHandle {referenceObject=null}
   impactNormal=<0 0 0>
   impactPoint=<0 0 0>
   location=<0 0 0>
   normal=<0 0 0>
   traceEnd=<0 0 0>
   traceStart=<0 0 0>
  }
 }
 TransactionContext {type=string, default=Snap To Target}
 VertOffsetAtStartOfVertMove {type=number}
 WheelTextHangTime {type=number, default=1}
 Widget Angle Rotated Line Label {type=string, default="Angle Rotated (deg)"}
 Widget Angle Rotated Line Value {type=string, default=0}
 Widget Distance Line Label {type=string, default="Distance (cm)"}
 Widget Distance Line Value {type=string, default=0}
]
